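# M_BugTestSimple is a translucent, Default Lit material for Niagara mesh particles, kept in the Glider_BentBaton glider's FX folder.
# Particle color RGB is halved, multiplied by a Fresnel term with zero base reflect fraction, and fed into MF_FX_EmissiveAdjustments as emissive.
# Particle alpha runs through its own Multiply node, and the graph uses no textures or exposed parameters.
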Material /BRCosmetics/Gadgets/Assets/VinderTech_GliderChute/Glider_BentBaton/FX/Materials/M_BugTestSimple
graph {
 properties {
  bAutomaticallySetUsageInEditor=false
  bUsedWithNiagaraMeshParticles=true
  blendMode=BLEND_Translucent
  previewMesh=/BRCosmetics/Gadgets/Assets/VinderTech_GliderChute/Glider_StunningMountain/FX/SM_Glider_StunningMountain_Sphere
  usedShadingModels=Default Lit
 }
 nodes [
  n1 = ParticleColor()
  MaterialFunctionCall(Emissive=Multiply(A=Multiply(A=@n1.RGB) {constB=0.5}, B=Fresnel() {baseReflectFraction=0})) {materialFunction=&MF_FX_EmissiveAdjustments}
  Multiply(A=@n1.A)
 ]
}
