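# Translucent, unlit, two-sided material for Niagara mesh particles that draws the glowing core effect of the Bent Baton glider.
# Emissive layers distorted core textures, a glow mask and panning noise in tintable colors, then applies Particle Color, which also scales the texture-driven opacity.
# A StepPanner reroute blends smooth and stepped time for choppy panning, while a Disturb reroute supplies noise UV distortion.
# Noise-driven local-space WPO, scaled by a Dynamic Parameter, fades at both U ends.
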Material /BRCosmetics/Gadgets/Assets/VinderTech_GliderChute/Glider_BentBaton/FX/Materials/M_Glider_BentBaton_Core
graph {
 properties {
  bAutomaticallySetUsageInEditor=false
  bUsedWithNiagaraMeshParticles=true
  blendMode=BLEND_Translucent
  shadingModel=MSM_Unlit
  twoSided=true
  usedShadingModels=Unlit
 }
 nodes [
  Constant2Vector() {r=-0.003}
  n1 = VectorParameter() {parameterName=WPO_MaskInfor, channelNames {a="", b=Highest, g=Lowest, r=Power}, defaultValue=<rgba 1 0.4 1 1>}
  n2 = Multiply(A=ComponentMask(None=TextureCoordinate()) {r=true}) {constB=4}
  LinearInterpolate(A=@n1.Lowest, B=@n1.Lowest, Alpha=Power(Base=Saturate(None=@n2), Exp=@n1.Power)) {constA=0.4}
  n3 = Add(A=TextureCoordinate()) {constB=0}
  n4 = TextureSample(UVs=Panner(Coordinate=TextureCoordinate() {vTiling=2}, Time=NamedRerouteUsage()) {speedX=-0.4}) {
   samplerType=SAMPLERTYPE_LinearColor
   texture=/Engine/Functions/Engine_MaterialFunctions02/ExampleContent/Textures/OffsetNoiseDistanceFields_ForNormals
  }
  n5 = Panner(Coordinate=TextureCoordinate() {uTiling=2, vTiling=1.5}, Time=NamedRerouteUsage()) {speedX=-0.4}
  n6 = LinearInterpolate(A=@n4.RGB, B=Multiply(A=TextureSample(UVs=@n5) {texture=/Game/Effects/Fort_Effects/Textures/Fire/HF_FireNoisesPacked}.RGB) {constB=1.5})
  n7 = ComponentMask(None=TextureCoordinate()) {r=true}
  n8 = Multiply(A=Multiply(A=ComponentMask(None=@n6) {g=true, r=true}, B=Power(Base=@n7) {constExponent=1.5})) {constB=0.1}
  n9 = Power(
   Base=TextureSample(UVs=Add(A=@n3, B=@n8)) {texture=&/T_Glider_BentBaton_Core}.R
   Exp=ScalarParameter() {defaultValue=0.7, group=Center02, parameterName=CenterTexColor02Power}
  ) {constExponent=0.7}
  n10 = ScalarParameter() {defaultValue=0.5, group=Center02, parameterName=CenterTexColor02Contrast}
  n11 = Multiply(
   A=VectorParameter() {group=Center02, parameterName=CenterTexColor02, defaultValue=<rgba 0.9323 0.0481 0.916 1>}.RGB
   B=Multiply(A=MaterialFunctionCall(In=@n9, Contrast=@n10) {materialFunction=&CheapContrast}.Result)
  )
  n12 = TextureSampleParameter2D() {parameterName=Lerp_Tex, texture=&/T_BentBaton_CoreGlow}
  n13 = TextureCoordinate()
  n14 = ComponentMask(None=TextureSample(UVs=Panner(Coordinate=@n13)) {texture=/BRCosmetics/Weapons/FORT_Melee/Pickaxe_SoapPocket/FX/T_Pickaxe_SoapPocket_Noise}.RGB) {g=true, r=true}
  n15 = Panner(Coordinate=Add(A=TextureCoordinate() {uTiling=1.5}, B=Multiply(A=@n14) {constB=0.2}), Time=NamedRerouteUsage()) {speedX=-0.5}
  n16 = LinearInterpolate(
   B=ScalarParameter() {defaultValue=0.5, parameterName=NoiseMax}
   Alpha=Power(Base=TextureSample(UVs=@n15) {samplerType=SAMPLERTYPE_Grayscale, texture=/BRCosmetics/Weapons/FORT_Melee/Pickaxe_Cube_Ninja_Male/FX/T_CellNoise}.R)
  ) {constB=0.5}
  n17 = Add(
   A=Multiply(
    A=VectorParameter() {parameterName=BlurColor, defaultValue=<rgba 0.6953 0.0351 1 1>}.RGB
    B=StaticComponentMaskParameter(None=@n12.RGB) {defaultR=true, parameterName=LerpTexMask}
   )
   B=Multiply(A=VectorParameter() {parameterName=OutSideNoiseColor, defaultValue=<rgba 0.2054 0.0354 0.5 1>}.RGB, B=Saturate(None=@n16))
  ) {constB=0}
  n18 = TextureCoordinate() {uTiling=0.5}
  n19 = NamedRerouteUsage()
  n20 = LinearInterpolate(Alpha=TextureSample(UVs=Panner(Coordinate=@n18, Time=@n19) {speedX=-0.5}) {samplerType=SAMPLERTYPE_Grayscale, texture=&LowResBlurredNoise}.R) {constB=5}
  n21 = VectorParameter() {parameterName=OutSideBaseColor, defaultValue=<rgba 0.0615 0 0.1146 1>}
  n22 = TextureSample(UVs=Add(A=TextureCoordinate(), B=NamedRerouteUsage())) {texture=&T_Glider_BentBaton_Core}
  n23 = ScalarParameter() {defaultValue=13, parameterName=CenterTexPower}
  n24 = LinearInterpolate(
   A=Add(A=Add(A=@n11, B=@n17), B=Multiply(A=@n20, B=@n21.RGB))
   B=VectorParameter() {parameterName=CenterTexColor, defaultValue=<rgba 1 0.6479 0.8917 1>}.RGB
   Alpha=Saturate(None=Power(Base=@n22.R, Exp=@n23) {constExponent=13})
  )
  n25 = ParticleColor()
  MaterialFunctionCall(Emissive=LinearInterpolate(A=Multiply(A=@n24, B=@n25.RGB), B=@n25.RGB, Alpha=ScalarParameter() {parameterName=UseSimpleGlow})) {materialFunction=&MF_FX_EmissiveAdjustments}
  n26 = TextureSampleParameter2D(UVs=Add(A=TextureCoordinate(), B=NamedRerouteUsage())) {
   group=Alpha
   parameterName=AlphaTex
   texture=&T_Glider_BentBaton_Core
  }
  n27 = Power(
   Base=StaticComponentMaskParameter(None=@n26.RGB) {defaultG=true, group=Alpha, parameterName=AlphaTexMask}
   Exp=ScalarParameter() {defaultValue=1, group=Alpha, parameterName=AlphaPower}
  ) {constExponent=1}
  n28 = ScalarParameter() {group=Alpha, parameterName=AlphaContrast}
  Multiply(
   A=@n25.A
   B=Multiply(A=MaterialFunctionCall(In=@n27, Contrast=@n28) {materialFunction=&CheapContrast}.Result)
  )
  n29 = VectorParameter() {group=WPO, parameterName=WPO, defaultValue=<rgba 0 20 0 1>}
  n30 = VectorParameter() {group=WPO, parameterName=WPO_A_UVControl, channelNames {a=VPan, b=UPan, g=VScale, r=UScale}, defaultValue=<rgba 1 1 -1 0>}
  n31 = TextureSample(UVs=MaterialFunctionCall!(UVInput=TextureCoordinate() {uTiling=0.7, vTiling=0}, TimeInput, Control_RGB=@n30, Control_A=@n30) {materialFunction=/Game/Effects/Niagara/MasterMaterials/Functions/Common/FXCommon_QuickUVControl}.UV) {texture=/BRCosmetics/Weapons/FORT_Melee/Pickaxe_MadameMoth/FX/T_Fibre_Noise}
  n32 = VectorParameter() {group=WPO, parameterName=WPO_B_UVControl, channelNames {a=VPan, b=UPan, g=VScale, r=UScale}, defaultValue=<rgba 1 1 -0.5 0>}
  n33 = TextureSample(UVs=MaterialFunctionCall(UVInput=TextureCoordinate() {uTiling=0.5, vTiling=0}, Control_RGB=@n32, Control_A=@n32) {materialFunction=&FXCommon_QuickUVControl}.UV) {texture=/Game/Weapons/WeaponSkins/Textures/T_Swirly_Noise_D1}
  n34 = ScalarParameter() {defaultValue=0.5, parameterName=WPO_TextureBlend}
  n35 = Multiply(A=Add(A=LinearInterpolate(A=@n31.R, B=@n33.R, Alpha=@n34), B=ScalarParameter() {group=WPO, parameterName=WPOOffset}))
  n36 = DynamicParameter() {paramNames [TimeRandom,WPOStr,KeyPanning,Offset]}
  n37 = TextureCoordinate()
  n38 = Multiply(A=Power(Base=ComponentMask(None=@n37) {r=true}) {constExponent=1}) {constB=3}
  n39 = ComponentMask(None=TextureCoordinate()) {r=true}
  n40 = Multiply(A=Power(Base=OneMinus(None=@n39)) {constExponent=3}) {constB=3}
  Multiply(
   A=Multiply(A=Transform(None=@n29.RGB) {transformSourceType=TRANSFORMSOURCE_Local}, B=Multiply(A=@n35, B=@n36.WPOStr))
   B=Multiply(A=Saturate(None=@n38), B=Saturate(None=@n40))
  )
  n41 = TextureSample(UVs=Panner(Coordinate=TextureCoordinate() {vTiling=2}, Time=NamedRerouteUsage()) {speedX=-0.4}) {
   samplerType=SAMPLERTYPE_LinearColor
   texture=&OffsetNoiseDistanceFields_ForNormals
  }
  n42 = ComponentMask(None=@n41.RGB) {g=true, r=true}
  n43 = Power(Base=ComponentMask(None=TextureCoordinate()) {r=true}, Exp=ScalarParameter() {defaultValue=2, parameterName=DistrurPower}) {constExponent=3}
  NamedRerouteDeclaration(None=Multiply(A=Multiply(A=@n42, B=@n43), B=ScalarParameter() {defaultValue=0.1, parameterName=DisturbIntensity}) {constB=0.2}) {name=Disturb, nodeColor=<rgba 1 0 0.4453 1>}
  n44 = Time()
  n45 = ScalarParameter() {defaultValue=1, parameterName=Speed}
  n46 = Multiply(A=Constant() {r=2.5}, B=@n45)
  n47 = Constant() {r=8}
  n48 = Ceil(None=Multiply(A=Time() {bOverride_Period=true, period=6000}, B=Multiply(A=@n47, B=@n45)) {constB=8})
  NamedRerouteDeclaration(None=LinearInterpolate(A=Multiply(A=@n44, B=@n46) {constB=2.3}, B=Multiply(A=@n48) {constB=0.5}, Alpha=ScalarParameter() {defaultValue=0.4, parameterName=LerpStep}) {constAlpha=1}) {name=StepPanner, nodeColor=<rgba 1 0.0368 0.001 1>}
 ]
}
