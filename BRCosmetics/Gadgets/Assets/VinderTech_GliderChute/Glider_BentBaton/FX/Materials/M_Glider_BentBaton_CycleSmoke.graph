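# Translucent, two-sided lit material for Niagara mesh particles, drawing purple cycling smoke for the Bent Baton glider FX.
# Opacity math combines view-angle fade, panning smoke and Perlin noise, and sine UV edge masks; color layers counter-panning smoke over radial gradients.
# Emissive lightning comes from plasma coils masked by smoke-modulated Fresnel, with noise flicker scaled by the LightningFlash dynamic parameter.
# Dynamic Width drives a SplineThicken offset, while noise-based normal displacement is multiplied by zero.
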Material /BRCosmetics/Gadgets/Assets/VinderTech_GliderChute/Glider_BentBaton/FX/Materials/M_Glider_BentBaton_CycleSmoke
graph {
 properties {
  bUsedWithNiagaraMeshParticles=true
  blendMode=BLEND_Translucent
  opacityMaskClipValue=0.1
  pixelDepthOffsetMode=PDOM_Legacy
  twoSided=true
  usedShadingModels=Default Lit
 }
 nodes [
  n1 = ParticleRandom()
  AppendVector(A=Constant(), B=@n1)
  n2 = ParticleColor()
  n3 = MaterialFunctionCall("FadeSensitivity (optional)"=ScalarParameter() {defaultValue=2, parameterName=Edge Fade}) {materialFunction=&FadeBasedOnViewAngle}
  n4 = VectorParameter() {parameterName=UIVScale, defaultValue=<rgba 1 0.4 0 0>}
  n5 = Add(A=Multiply(A=TextureCoordinate(), B=ComponentMask(None=@n4.RGB) {g=true, r=true}), B=AppendVector(A=Constant(), B=ParticleRandom()))
  n6 = LinearInterpolate(Alpha=TextureSample(UVs=Panner(Coordinate=@n5) {speedX=0.5}) {texture=&T_Inky_Smoke_Tile}.R) {constA=3, constB=10}
  n7 = Saturate(None=Multiply(B=Power(Base=@n3.Result, Exp=@n6)) {constA=10})
  n8 = Sine(None=ComponentMask(None=TextureCoordinate()) {g=true}) {period=2}
  n9 = Multiply(A=Power(Base=@n8)) {constB=3}
  n10 = VectorParameter() {parameterName=UIVScale, defaultValue=<rgba 1 0.4 0 0>}
  n11 = Panner(Coordinate=Multiply(A=TextureCoordinate(), B=ComponentMask(None=@n10.RGB) {g=true, r=true})) {speedX=0.5}
  n12 = ScalarParameter() {defaultValue=5, parameterName=UpAndDownMask}
  n13 = OneMinus(None=@n8)
  n14 = Saturate(None=Power(
    Base=TextureSample(UVs=@n11) {samplerType=SAMPLERTYPE_LinearColor, texture=/Game/Global/Textures/Default/Tileable_Patterns/T_PerlinNoise_Uncompressed}.R
    Exp=LinearInterpolate(B=@n12, Alpha=@n13) {constB=5}
   ))
  n15 = LinearInterpolate(B=LinearInterpolate(Alpha=Multiply(A=@n9, B=@n14)), Alpha=ScalarParameter() {defaultValue=1, parameterName=Use_SecondMask})
  n16 = Sine(None=ComponentMask(None=TextureCoordinate()) {r=true}) {period=2}
  n17 = VectorParameter() {parameterName=UIVScale, defaultValue=<rgba 1 0.4 0 0>}
  n18 = Add(A=Multiply(A=TextureCoordinate(), B=ComponentMask(None=@n17.RGB) {g=true, r=true}), B=AppendVector(A=Constant(), B=ParticleRandom()))
  n19 = LinearInterpolate(Alpha=TextureSample(UVs=Panner(Coordinate=@n18) {speedX=0.5}) {texture=&T_Inky_Smoke_Tile}.R) {constA=3, constB=10}
  n20 = Saturate(None=Multiply(A=Power(Base=@n16, Exp=@n19) {constExponent=3}) {constB=3})
  n21 = Multiply(A=Multiply(A=Multiply(A=@n7), B=Multiply(A=@n15, B=@n20)))
  MaterialFunctionCall(Number=Clamp(None=Multiply(A=@n2.A, B=@n21)) {maxDefault=0.999}) {materialFunction=&DebugScalarValues}
  n22 = Add(A=TextureCoordinate(), B=ParticleRandom())
  n23 = Multiply(
   A=TextureSample(UVs=Panner(Coordinate=@n22) {speedX=0.5, speedY=1}) {samplerType=SAMPLERTYPE_Grayscale, texture=&LowResBlurredNoise}.R
   B=Multiply(
    A=ScalarParameter() {defaultValue=15, parameterName=WPO Intensity}
    B=DynamicParameter() {paramNames [WPO Intensity, BrightOffset, LightningFlash, Width]}.WPO Intensity
   )
  ) {constB=25}
  n24 = VertexNormalWS()
  NamedRerouteDeclaration(None=Multiply(A=Multiply(A=@n23, B=@n24) {constB=0})) {name=WPO, nodeColor=<rgba 1 0 0.7734 1>}
  n25 = MaterialFunctionCall(Power=Constant() {r=1.5}) {materialFunction=&Fresnel_Function}
  NamedRerouteDeclaration(None=LinearInterpolate(
    A=VectorParameter() {parameterName=ColorCenter, defaultValue=<rgba 0.2334 0.0432 0.5 1>}.RGB
    B=VectorParameter() {parameterName=ColorOutside, defaultValue=<rgba 0.1461 0.0541 0.5 1>}.RGB
    Alpha=Multiply(A=@n25.Result)
   )) {name=Fresnel, nodeColor=<rgba 0 0.0625 0.5208 1>}
  OneMinus(None=SphereMask(A=TextureCoordinate(), B=Constant() {r=0.5}) {attenuationRadius=1, hardnessPercent=0})
  Reroute(None=Add(A=NamedRerouteUsage()) {constB=0})
  n26 = TextureCoordinate()
  n27 = ComponentMask(None=VectorParameter() {parameterName=LightningUVScale, defaultValue=<rgba 1 0.25 0 0>}.RGB) {g=true, r=true}
  n28 = Add(A=Add(A=Multiply(A=@n26, B=@n27), B=@n1) {constB=0}) {constB=0.38}
  n29 = Multiply(
   A=TextureCoordinate()
   B=ComponentMask(None=VectorParameter() {parameterName=LightningUVScale, defaultValue=<rgba 1 0.25 0 0>}.RGB) {g=true, r=true}
  )
  n30 = TextureSample(UVs=Panner(Coordinate=Add(A=@n29, B=@n1)) {speedX=0.5}) {texture=&T_Inky_Smoke_Tile}
  n31 = LinearInterpolate(Alpha=@n30.R) {constA=0.3}
  n32 = Multiply(
   A=TextureSample(UVs=Panner(Coordinate=@n28) {speedX=-0.5}) {samplerType=SAMPLERTYPE_LinearGrayscale, texture=/Game/Effects/Fort_Effects/Emotes/UltraLight/Texture/T_PlasmaCoils}.R
   B=@n31
  )
  n33 = Multiply(
   A=MaterialFunctionCall(Invert Fresnel=StaticBool() {value=true}, Power=LinearInterpolate(Alpha=@n30.R) {constA=3, constB=20}) {materialFunction=&Fresnel_Function}.Result
   B=MaterialFunctionCall(Invert Fresnel=StaticBool(), Power=LinearInterpolate(Alpha=@n30.R) {constA=0.5}) {materialFunction=&Fresnel_Function}.Result
  )
  n34 = Power(
   Base=Multiply(A=Power(Base=@n32) {constExponent=3}, B=Multiply(A=@n31, B=@n33))
   Exp=ScalarParameter() {defaultValue=1, parameterName=LightningPower}
  ) {constExponent=3}
  n35 = Panner(Coordinate=Add(
    A=TextureCoordinate() {uTiling=0, vTiling=0}
    B=DynamicParameter() {paramNames [WPO Intensity, BrightOffset, LightningFlash, Width]}.BrightOffset
   )) {speedX=0.07, speedY=0.07}
  n36 = Saturate(None=LinearInterpolate(Alpha=TextureSample(UVs=@n35) {texture=/Game/Global/Textures/Default/Noise/Noise}.R) {constA=-1, constB=3})
  n37 = MaterialFunctionCall(Speed=Constant() {r=1}) {materialFunction=&DebugTimeSine}
  n38 = Multiply(
   A=Multiply(A=VectorParameter() {parameterName=Color Lighting, defaultValue=<rgba 994.7936 0.0004 3000 1>}.RGB, B=Saturate(None=@n34))
   B=Multiply(
    A=LinearInterpolate(B=@n36, Alpha=@n37.Result)
    B=DynamicParameter() {paramNames [WPO Intensity, BrightOffset, LightningFlash, Width]}.LightningFlash
   )
  )
  n39 = VectorParameter() {parameterName=BaseEmissiveColor, defaultValue=<rgba 0 0 0 1>}
  Reroute(None=MaterialFunctionCall(Emissive=Add(A=@n38, B=@n39.RGB)) {materialFunction=&MF_FX_EmissiveAdjustments}.Result)
  n40 = TextureCoordinate()
  n41 = Constant() {r=0.5}
  n42 = ScalarParameter() {defaultValue=0.7, parameterName=Radius}
  n43 = ScalarParameter() {parameterName=Hardness}
  n44 = LinearInterpolate(
   A=LinearInterpolate(
    A=VectorParameter() {parameterName=ColorOutside, defaultValue=<rgba 0.1461 0.0541 0.5 1>}.RGB
    B=VectorParameter() {parameterName=ColorCenter, defaultValue=<rgba 0.2334 0.0432 0.5 1>}.RGB
    Alpha=SphereMask(A=@n40, B=@n41, Radius=@n42, Hardness=@n43) {attenuationRadius=0.7, hardnessPercent=0}
   )
   B=VectorParameter() {parameterName=ColorCenter02, defaultValue=<rgba 0.0104 0.0151 0.1 1>}.RGB
   Alpha=SphereMask(A=TextureCoordinate(), B=Constant() {r=0.5}) {attenuationRadius=0.3, hardnessPercent=45}
  )
  n45 = ScalarParameter() {defaultValue=1.5, parameterName=BaseNoiseMax}
  n46 = Panner(Coordinate=TextureCoordinate() {uTiling=0.7, vTiling=0.4})
  n47 = Power(
   Base=DotProduct(
    A=Constant3Vector() {constant=<rgba 0 0 1 0>}
    B=TextureSample(UVs=@n46) {samplerType=SAMPLERTYPE_Normal, texture=/Game/Accessories/FORT_Backpacks/Backpack_M_MED_EctoCat/FX/CloudNoise_T_N}.RGB
   )
   Exp=ScalarParameter() {defaultValue=5, parameterName=BaseNoisePower}
  ) {constExponent=5}
  n48 = VectorParameter() {parameterName=ColorNoise, defaultValue=<rgba 0.2043 0.0646 0.4 1>}
  n49 = Multiply(A=TextureCoordinate(), B=ComponentMask(None=VectorParameter() {parameterName=UIVScale, defaultValue=<rgba 1 0.4 0 0>}.RGB) {g=true, r=true})
  n50 = Constant2Vector() {g=0.5, r=0.5}
  n51 = Constant()
  n52 = ParticleRandom()
  n53 = Panner(Coordinate=Add(A=Add(A=@n49, B=@n50), B=AppendVector(A=@n51, B=@n52))) {speedX=-0.5}
  n54 = Constant()
  n55 = ParticleRandom()
  n56 = Panner(Coordinate=Add(A=@n49, B=AppendVector(A=@n54, B=@n55))) {speedX=0.5}
  n57 = Saturate(None=Add(
    A=TextureSample(UVs=@n53) {texture=&T_Inky_Smoke_Tile}.R
    B=TextureSample(UVs=@n56) {texture=&T_Inky_Smoke_Tile}.R
   ))
  n58 = Add(
   A=Multiply(A=ParticleColor().RGB) {constB=0.01}
   B=Multiply(A=LinearInterpolate(B=@n45, Alpha=@n47) {constB=1.5}, B=Multiply(A=@n48.RGB, B=@n57)) {constA=1}
  ) {constB=0}
  n59 = ParticleColor()
  n60 = ScalarParameter() {defaultValue=1, parameterName=UseParticleColortoBase}
  Reroute(None=Multiply(A=Add(A=@n58, B=@n44) {constB=0}, B=LinearInterpolate(B=@n59.RGB, Alpha=@n60) {constA=1}))
  n61 = Multiply(A=Constant() {r=20}, B=DynamicParameter() {paramNames [WPO Intensity, BrightOffset, LightningFlash, Width]}.Width)
  Reroute(None=MaterialFunctionCall(WidthBase=@n61, WidthTip=@n61, Expand U or V UV Channel=StaticBoolParameter() {parameterName="Expand U or V UV Channel (SB)"}) {materialFunction=&SplineThicken}.WorldPositionOffset)
  Reroute(None=Add(A=NamedRerouteUsage()) {constB=0})
  Saturate(None=@n31)
  TextureSample(UVs=TextureCoordinate() {uTiling=2, vTiling=0.7}) {samplerType=SAMPLERTYPE_Grayscale, texture=&LowResBlurredNoise}
 ]
}
