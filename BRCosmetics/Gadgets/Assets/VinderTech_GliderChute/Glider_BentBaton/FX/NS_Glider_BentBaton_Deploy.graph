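# Niagara deploy effect for the BentBaton glider, showing purple-magenta clouds, layered glow meshes with a light, and lightning.
# Four emitters attach to the glider's skeletal mesh using shared system position, timing and size values.
# Two-partition clouds vary velocity, drag and scale from smoothed system interpolation values, while lightning arcs and split bolts animate through dynamic material parameters.
# Front-end checks adjust delay, brightness and scale for preview versus gameplay.
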NiagaraSystem /BRCosmetics/Gadgets/Assets/VinderTech_GliderChute/Glider_BentBaton/FX/NS_Glider_BentBaton_Deploy
dependencies {
 usedDataInterfaces [
  &NiagaraDataInterfaceCurve
  &NiagaraDataInterfaceArrayFloat3
  &NiagaraDataInterfaceVectorCurve
  &NiagaraDataInterfaceColorCurve
  &NiagaraDataInterfaceArrayFloat
  &NiagaraDataInterfaceSkeletalMesh
 ]
 usedDynamicInputs [
  &FN_FrontEndVector3Modifier
  &Subtract_Float
  &Abs_Float
  &DotProduct
  /Niagara/DynamicInputs/Multiply/ScaleAndBiasFloat
  &Ceil_Float
  &SetBoolByFloatComparison
  &ClampFloat
 ]
 usedModules [
  &self:SystemSpawnScript.SystemScriptSource.SystemScriptGraph.NiagaraNodeAssignment_0.SetVariables_#1
  &self:SystemSpawnScript.SystemScriptSource.SystemScriptGraph.NiagaraNodeAssignment_1.SetVariables_#2
  &self:SystemSpawnScript.SystemScriptSource.SystemScriptGraph.NiagaraNodeAssignment_5.SetVariables_#3
  &self:SystemSpawnScript.SystemScriptSource.SystemScriptGraph.NiagaraNodeAssignment_2.SetVariables_#4
  &self:SystemSpawnScript.SystemScriptSource.SystemScriptGraph.NiagaraNodeAssignment_6.SetVariables_#5
  &self:SystemSpawnScript.SystemScriptSource.SystemScriptGraph.NiagaraNodeAssignment_3.SetVariables_#6
  &SystemState
  &InterpolateOverTime
  &self:SystemSpawnScript.SystemScriptSource.SystemScriptGraph.NiagaraNodeAssignment_4.SetVariables_#7
 ]
}
emitters {
 Cloud001 {
  emitterUpdateScript [
   &EmitterState(
    Life Cycle Mode=::NewEnumerator1
    Loop Behavior=::NewEnumerator1
    Loop Delay=&FN_FrontEndFloatModifier(Gameplay=0, Front End=0.3)
    Loop Duration=$System.LoopDuration
    UseLoopDelay=true
   )
   &SpawnBurst_Instantaneous(Spawn Count=10, Use Spawn Probability=true, UseLoopCountLimit=false)
   &SpawnRate(SpawnRate=30, Use Spawn Probability=true)
   &SpawnBurst_Instantaneous(Spawn Count=10, Spawn Time=2.4833, Use Spawn Probability=false, UseLoopCountLimit=false) {module=SpawnBurst_Instantaneous001}
   /SetVariables(Emitter.BaseEmissiveColor:LinearColor=&MakeCustomLinearColorFromBool(Bool=$User.bIsFrontEndPreview, True Color=<rgba 0.0265 0.0045 0.05 1>, False Color=<rgba 0 0 0 1>)) {isSetParameters=true}
  ]
  particleSpawnScript [
   &PartitionParticles()
   &InitializeParticle(
    Color=<rgba 0.8455 0.8455 0.8455 1>
    Lifetime Max=0.9
    Lifetime Min=0.7
    Lifetime Mode=::NewEnumerator1
    Mesh Scale=&Multiply_VectorByFloat(Vector=&RandomRangeVector(Minimum=<3 2 2>, Maximum=<4 3 3>), Float=0.5)
    Mesh Scale Mode=::NewEnumerator4
    UsePositionOffset=false
    Write Material Random=true
   )
   &SkeletalMeshLocation(
    Bone Sampling Mode=::NewEnumerator4
    Interpolate Mesh Sampling=false
    MaskSpawnGroup=false
    Orientation Sampling=::NewEnumerator1
    Sampled Position Offset=&AddVectorToPosition(
     Position=&SelectVectorFromArray(
      Direct Array Index=$Particles.PartitionParticles.Partition
      Vector Selection Array=DataInterface() {propertyValues {FloatData [<-25 -20 -5>, <20 -20 -5>]}}
     )
     Vector=$System.FX position offset
    )
    Skeletal Mesh=$System.Skeletal Mesh
   )
   &AddVelocity(
    Scale Added Velocity=&Multiply_VectorByFloat(
     Vector=&Multiply_VectorByFloat(Vector=<2.5 2.5 2.5>, Float=1)
     Float=&Lerp_Float(A=1, B=0.6, Alpha=$System.InterpolateOverTime.InterpolateOverTime_Float.Moving Average)
    )
    Velocity=&SelectVectorFromArray(
     Direct Array Index=$Particles.PartitionParticles.Partition
     Vector Selection Array=DataInterface() {propertyValues {FloatData [<-25 120 0>, <25 160 0>]}}
    )
   )
   &AddVelocityInCone(Cone Angle=10, Cone Axis=<0 1 0>, Module.Use Velocity Falloff On Cone Axis=false, Velocity Strength=100) {module=AddVelocityInCone001}
   &InitialMeshOrientation(
    Mesh Orientation Mode=::NewEnumerator2
    Rotation=&RandomRangeVector(Minimum=<0 0 0>, Maximum=<1 0 0>)
    Use Rotation Vector=true
   )
   &CameraOffset()
  ]
  particleUpdateScript [
   &Drag(
    Drag=&Multiply_Float(
     A=&MakeCustomFloatFromBool(
      Bool=&SetBoolByIntComparison(A=$Particles.PartitionParticles.Partition, B=0, Comparison Type=::NewEnumerator0)
      True Float=&Lerp_Float(A=0, B=1, Alpha=$System.InterpolateOverTime001.InterpolateOverTime_Float.Moving Average)
      False Float=&Lerp_Float(A=0, B=1, Alpha=$System.InterpolateOverTime002.InterpolateOverTime_Float.Moving Average)
     )
     B=1
    )
    Use Linear Drag=true
    Use Rotational Drag=true
   )
   &SolveForcesAndVelocity(Clamp Velocity=false, Limit Acceleration=false)
   &ParticleState()
   &ScaleMeshSize(Scale Factor=&Multiply_VectorByFloat(
     Vector=&Multiply_VectorByFloat(
      Vector=&Multiply_Vector(
       A=<1 1 1>
       B=&VectorFromCurve(VectorCurve=DataInterface() {propertyValues {XCurve {keys [<at 0 5>,<at 0.2 1>,<at 1 1>]}, YCurve {keys [<at 0 1>,<at 1 1>]}, ZCurve {keys [<at 0 1>,<at 1 1>]}}})
      )
      Float=&RandomRangeFloat(Minimum=1, Maximum=0.7)
     )
     Float=&FloatFromCurve(
      FloatCurve=DataInterface() {propertyValues {Curve {keys [
          {interpMode=RCIM_Cubic, tangentMode=RCTM_User, time=0, value=0.3}
          {arriveTangent=1.3433, tangentMode=RCTM_User, time=0.1866, value=0.3318}
          {arriveTangent=-1.0223, interpMode=RCIM_Cubic, leaveTangent=-1.0223, tangentMode=RCTM_User, time=1, value=0.8}
         ]}}}
      Scale Curve=&Multiply_Float(
       A=&Multiply_Float(
        A=&Multiply_Float(
         A=1.3
         B=&FloatFromCurve(FloatCurve=DataInterface() {propertyValues {Curve {keys [<at 0.5 0>,<at 0.6 1>,<at 1 1>]}}}, CurveIndex=$Emitter.Age)
        )
        B=&MakeCustomFloatFromBool(
         Bool=&SetBoolByIntComparison(A=$Particles.PartitionParticles.Partition, B=0, Comparison Type=::NewEnumerator0)
         True Float=&Lerp_Float(A=1, B=0.7, Alpha=$System.InterpolateOverTime001.InterpolateOverTime_Float.Moving Average)
         False Float=&Lerp_Float(A=1, B=0.7, Alpha=$System.InterpolateOverTime002.InterpolateOverTime_Float.Moving Average)
        )
       )
       B=&Lerp_Float(A=1, B=0.7, Alpha=$System.InterpolateOverTime.InterpolateOverTime_Float.Moving Average)
      )
     )
    ))
   &DynamicMaterialParameters(
    Index 0 Param 1=&FloatFromCurve(
     FloatCurve=DataInterface() {propertyValues {Curve {keys [
         {interpMode=RCIM_Cubic, tangentMode=RCTM_User, time=0, value=0}
         <at 0.2 0.55>
         {arriveTangent=-1.0223, interpMode=RCIM_Cubic, leaveTangent=-1.0223, tangentMode=RCTM_User, time=1, value=1}
        ]}}}
     Scale Curve=&RandomRangeFloat(Minimum=0.7, Maximum=1.5)
    )
    Index 0 Param 3=&Lerp_Float(A=0, B=1, Alpha=$System.InterpolateOverTime004.InterpolateOverTime_Float.Moving Average)
    Module.Index0Param0Write=true
    Module.Index0Param1Write=true
    Module.Index0Param2Write=true
    Module.Index0Param3Write=true
   )
   &ScaleColor(
    Scale Alpha=&FloatFromCurve(
     FloatCurve=DataInterface() {propertyValues {Curve {keys [
         {arriveTangent=2.8927, interpMode=RCIM_Cubic, leaveTangent=2.8927, tangentMode=RCTM_User, time=0, value=1}
         <at 0.1 1 RCIM_Cubic>
         <at 0.3 1>
         <at 0.8 1>
         <at 1 0>
        ]}}}
     Scale Curve=&FloatFromCurve(FloatCurve=DataInterface() {propertyValues {Curve {keys [<at 2.6 1>,<at 2.8 0>]}}}, CurveIndex=$Emitter.Age)
    )
    Scale RGB=&Multiply_VectorByFloat(
     Vector=&MakeVectorFromLinearColorRGB(Color=&ColorFromCurve(ColorCurve=DataInterface() {propertyValues {
         AlphaCurve {keys []}
         BlueCurve {keys [{arriveTangent=-331.6383, interpMode=RCIM_Cubic, time=0, value=3.9554}, <at 0.3949 5>, <at 0.7598 3 RCIM_Cubic>]}
         GreenCurve {keys [{arriveTangent=-87.3695, interpMode=RCIM_Cubic, time=0, value=0}, <at 0.3949 0>, <at 0.7598 0.6094 RCIM_Cubic>]}
         RedCurve {keys [<at 0 5 RCIM_Cubic>,<at 0.3949 1.0461>,<at 0.7598 0.6906 RCIM_Cubic>]}
        }}))
     Float=&FN_FrontEndFloatModifier(Gameplay=1, Front End=1.5)
    )
    ScaleA=true
    ScaleRGB=true
   )
   &FN_Mesh_Attributes(
    Color=&MakeLinearColorFromVectorAndFloat(
     "Vector (RGB)"=&Multiply_Vector(
      A=&MakeVectorFromLinearColorRGB(Color=<rgba 1 1 1 1>)
      B=&MakeVectorFromLinearColorRGB(Color=&ColorFromCurve(ColorCurve=DataInterface() {propertyValues {
          AlphaCurve {keys []}
          BlueCurve {keys [{arriveTangent=-331.6383, interpMode=RCIM_Cubic, time=0, value=3.9554}, <at 0.3949 5>, <at 0.7598 3 RCIM_Cubic>]}
          GreenCurve {keys [{arriveTangent=-87.3695, interpMode=RCIM_Cubic, time=0, value=0}, <at 0.3949 0>, <at 0.7598 0 RCIM_Cubic>]}
          RedCurve {keys [<at 0 5 RCIM_Cubic>,<at 0.3949 1.0461>,<at 0.7598 0.1019 RCIM_Cubic>]}
         }}))
     )
     "Float (Alpha)"=&Multiply_Float(
      A=&FloatFromCurve(FloatCurve=DataInterface() {propertyValues {Curve {keys [{arriveTangent=2.8927, interpMode=RCIM_Cubic, leaveTangent=2.8927, tangentMode=RCTM_User, time=0, value=0}, <at 0.3 1>, <at 0.8 1>, <at 1 0>]}}})
      B=1
     )
    )
    Create Color Attribute=true
    Create Scale Attribute=true
    Scale=&Multiply_Vector(
     A=<0.7 0.7 0.7>
     B=&Multiply_VectorByFloat(
      Vector=&Multiply_VectorByFloat(
       Vector=&Multiply_Vector(
        A=<1 1 1>
        B=&VectorFromCurve(
         VectorCurve=DataInterface() {propertyValues {XCurve {keys [<at 0 5>,<at 0.2 1>,<at 1 1>]}, YCurve {keys [<at 0 1>,<at 1 1>]}, ZCurve {keys [<at 0 1>,<at 1 1>]}}}
         Scale Curve=&Multiply_VectorByFloat(
          Vector=<1 1 1>
          Float=&FloatFromCurve(FloatCurve=DataInterface() {propertyValues {Curve {keys [<at 0.5 0>,<at 0.6 1>,<at 1 1>]}}}, CurveIndex=$Emitter.Age, Scale Curve=0.7)
         )
        )
       )
       Float=&RandomRangeFloat(Minimum=1, Maximum=0.7)
      )
      Float=&FloatFromCurve(
       FloatCurve=DataInterface() {propertyValues {Curve {keys [
           {interpMode=RCIM_Cubic, tangentMode=RCTM_User, time=0, value=0.2}
           {arriveTangent=1.3433, tangentMode=RCTM_User, time=0.2, value=0.3}
           <at 0.8 0.7>
           {arriveTangent=-1.0223, interpMode=RCIM_Cubic, leaveTangent=-1.0223, tangentMode=RCTM_User, time=1, value=0}
          ]}}}
       Scale Curve=&Multiply_Float(
        A=&Multiply_Float(
         A=1.5
         B=&MakeCustomFloatFromBool(
          Bool=&SetBoolByIntComparison(A=$Particles.PartitionParticles.Partition, B=0, Comparison Type=::NewEnumerator0)
          True Float=&Lerp_Float(A=1, B=0.7, Alpha=$System.InterpolateOverTime001.InterpolateOverTime_Float.Moving Average)
          False Float=&Lerp_Float(A=1, B=0.7, Alpha=$System.InterpolateOverTime002.InterpolateOverTime_Float.Moving Average)
         )
        )
        B=&Lerp_Float(A=1, B=0.7, Alpha=$System.InterpolateOverTime.InterpolateOverTime_Float.Moving Average)
       )
      )
     )
    )
    Write to Module Namespace=true
   )
  ]
  rendererClasses [&NiagaraMeshRendererProperties]
  renderers [{rendererClass=&NiagaraMeshRendererProperties, rendererIndex=0}]
 }
 FXMesh001 {
  emitterUpdateScript [
   &EmitterState(
    Life Cycle Mode=::NewEnumerator1
    Loop Behavior=::NewEnumerator1
    Loop Delay=$System.Delay
    Loop Duration=$System.LoopDuration
    UseLoopDelay=true
   )
   &SpawnBurst_Instantaneous(Use Spawn Probability=false, UseLoopCountLimit=false)
  ]
  particleSpawnScript [
   &InitializeParticle(
    Color Mode=::NewEnumerator0
    Lifetime=2
    Mesh Scale Mode=::NewEnumerator3
    Mesh Uniform Scale=3
    Position Offset=<0 0 -10>
    UsePositionOffset=false
    Write Material Random=true
   )
   &InitialMeshOrientation(Mesh Orientation Mode=::NewEnumerator2, Rotation=<0 0.25 0>, Use Rotation Vector=true)
  ]
  particleUpdateScript [
   &SkeletalMeshLocation(
    Bone Sampling Mode=::NewEnumerator4
    MaskSpawnGroup=false
    Sampled Position Offset=&AddVectorToPosition(Position=Linked() {linkedVariable {name=System.FXMeshPosition, type=Vector3f}}, Vector=<0 0 0>)
    Skeletal Mesh=$System.Skeletal Mesh
   )
   &ParticleState()
   &ScaleMeshSize(Scale Factor=&Multiply_VectorByFloat(Vector=<1 1.5 1.3>, Float=$System.SizeScale))
   &FN_Mesh_Attributes(
    Color=&MakeLinearColorFromVectorAndFloat(
     "Vector (RGB)"=&MakeVectorFromLinearColorRGB(Color=<rgba 0.9663 0.0382 1 0.5>)
     "Float (Alpha)"=&FN_FrontEndFloatModifier(Gameplay=0.5, Front End=0.3)
    )
    Create Color Attribute=true
    Create Scale Attribute=true
    Scale=&Multiply_VectorByFloat(
     Vector=&Multiply_Vector(A=<1 1.5 1>, B=&VectorFromFloat(Value=$System.SizeScale))
     Float=&FN_FrontEndFloatModifier(Gameplay=4, Front End=3)
    )
    Write to Module Namespace=true
   )
   &FN_Mesh_Attributes(
    Color=&MakeLinearColorFromVectorAndFloat("Vector (RGB)"=&MakeVectorFromLinearColorRGB(Color=<rgba 0.1093 0.0045 0.2 1>), "Float (Alpha)"=1)
    Create Color Attribute=true
    Create Scale Attribute=true
    Scale=&Multiply_VectorByFloat(
     Vector=&Multiply_Vector(A=<1.7 1.5 1>, B=&VectorFromFloat(Value=$System.SizeScale))
     Float=0.4
    )
    Write to Module Namespace=true
   ) {module=FN_Mesh_Attributes001}
   &FN_Mesh_Attributes(
    Color=&MakeLinearColorFromVectorAndFloat(
     "Vector (RGB)"=&MakeVectorFromLinearColorRGB(Color=<rgba 4 0.0084 3.5375 1>)
     "Float (Alpha)"=&FN_FrontEndFloatModifier(Gameplay=1, Front End=0)
    )
    Create Color Attribute=true
    Create Scale Attribute=true
    Scale=&Multiply_VectorByFloat(
     Vector=&Multiply_Vector(A=<1.7 1.5 1>, B=&VectorFromFloat(Value=$System.SizeScale))
     Float=0.2
    )
    Write to Module Namespace=true
   ) {module=FN_Mesh_Attributes002}
   &Light_Attributes(
    Color=1
    Color Mode=0
    Falloff Exponent=2
    Light Color:LinearColor=&Multiply_LinearColorByFloat(Linear Color=<rgba 29.8229 2.2782 30 1>, Float=1.5)
    Light Exponent=3
    Light Position Offset:Vector3f=<10 100 0>
    Light Radius=150
    Position=1
    Position Offset Coordinate Space:ENiagaraCoordinateSpace=::Local
    Radius=2
    UsePositionOffset=true
   )
  ]
  rendererClasses [&NiagaraMeshRendererProperties, &NiagaraLightRendererProperties]
  renderers [
   {rendererClass=&NiagaraMeshRendererProperties, rendererIndex=0}
   {rendererClass=&NiagaraMeshRendererProperties, rendererIndex=1}
   {rendererClass=&NiagaraMeshRendererProperties, rendererIndex=2}
   {rendererClass=&NiagaraMeshRendererProperties, rendererIndex=3}
   {rendererClass=&NiagaraLightRendererProperties, rendererIndex=4}
  ]
 }
 Lightning_Arc {
  emitterUpdateScript [
   &EmitterState(Life Cycle Mode=::NewEnumerator1, Loop Behavior=::NewEnumerator1, Loop Delay=$System.Delay, Loop Duration=2.2, UseLoopDelay=true)
   &SpawnBurst_Instantaneous(Use Spawn Probability=true, UseLoopCountLimit=false)
   &SpawnRate(SpawnRate=7, Use Spawn Probability=true)
  ]
  particleSpawnScript [
   &InitializeParticle(
    Color=<rgba 0.5 0 0.3068 10>
    Lifetime Max=0.7
    Lifetime Min=0.2
    Lifetime Mode=::NewEnumerator1
    Mesh Scale=&RandomRangeVector(Minimum=<1 2 1.3>, Maximum=<1 1.5 1>)
    Mesh Scale Mode=::NewEnumerator4
    UsePositionOffset=false
    Write Material Random=true
   )
   &CameraOffset()
   &ShapeLocation(
    Cylinder Height=0
    Cylinder Radius=1
    MaskSpawnGroup=false
    Override Seed=false
    Rotation Axis=<0 1 0>
    Rotation Mode=::NewEnumerator1
    Shape Primitive=::NewEnumerator1
    Surface Only=false
    UseOwnerScale=true
   )
   &AddVelocity(Module.Use Falloff Distance=false, Override Seed=false, Velocity Mode=::NewEnumerator1, Velocity Speed=1)
   /SetVariables(Particles.Random=&RandomRangeFloat(Minimum=0, Maximum=1)) {isSetParameters=true}
   &InitialMeshOrientation(
    Mesh Orientation Mode=::NewEnumerator2
    Rotation=&MakeCustomVectorFromBool(
     Bool=&RandomBool()
     True Vector=&RandomRangeVector(Minimum=<0 -0.13 0>, Maximum=<0 0.13 0>)
     False Vector=&RandomRangeVector(Minimum=<0 0.37 0>, Maximum=<0 0.63 0>)
    )
    Use Rotation Vector=true
   ) {module=InitialMeshOrientation001}
   &InitialMeshOrientation(Mesh Orientation Mode=::NewEnumerator2, Use Rotation Vector=true) {module=InitialMeshOrientation003}
  ]
  particleUpdateScript [
   &SkeletalMeshLocation(
    Bone Sampling Mode=::NewEnumerator4
    MaskSpawnGroup=false
    Sampled Position Offset=&AddVectorToPosition(Position=$System.FXMeshPosition, Vector=&RandomRangeVector(Minimum=<0 -100 0>, Maximum=<0 -250 0>))
    Skeletal Mesh=$System.Skeletal Mesh
   )
   &ParticleState()
   &ScaleMeshSize(Scale Factor=&Multiply_VectorByFloat(Vector=<3 3 3>, Float=&RandomRangeFloat(Minimum=0.8, Maximum=1.5)))
   &ScaleColor(
    Scale Alpha=&Multiply_Float(
     A=10
     B=&FloatFromCurve(FloatCurve=DataInterface() {propertyValues {Curve {keys [<at 0 0>,<at 0.1 0>,<at 0.2 0.8>,<at 1 0>]}}})
    )
    Scale RGB=&Multiply_VectorByFloat(
     Vector=<1 1 1>
     Float=&Multiply_Float(
      A=&RandomRangeFloat(Minimum=0, Maximum=1, Evaluation Type=::NewEnumerator1)
      B=&Multiply_Float(
       A=1
       B=&FloatFromCurve(FloatCurve=DataInterface() {propertyValues {Curve {keys [<at 0 10>,<at 1 10>]}}})
      )
     )
    )
    ScaleA=true
    ScaleRGB=true
   )
   &DynamicMaterialParameters(
    Index 0 Param 1=&FloatFromCurve(FloatCurve=DataInterface() {propertyValues {Curve {keys [{tangentMode=RCTM_User, time=0, value=0}, {tangentMode=RCTM_User, time=1, value=1}]}}})
    Index 0 Param 2=0.7
    Index 0 Param 3=&FloatFromCurve(FloatCurve=DataInterface() {propertyValues {Curve {keys [<at 0 0 RCIM_Cubic>, {arriveTangent=-0.3684, interpMode=RCIM_Cubic, leaveTangent=-0.3684, tangentMode=RCTM_User, time=1, value=3}]}}})
    Index 0 Param 4=2
    Index 1 Param 1=&FN_FrontEndFloatModifier(Gameplay=1.5, Front End=0.7)
    Index 1 Param 3=&Lerp_Float(A=1, B=1.5, Alpha=$Particles.Random)
    Index 1 Param 4=1
    Module.Param0WriteEnabled=true
    Module.Param0WriteEnabled1=true
    Module.Param1WriteEnabled=true
    Module.Param1WriteEnabled1=true
    Module.Param2WriteEnabled=true
    Module.Param2WriteEnabled1=true
    Module.Param3WriteEnabled=true
    Module.Param3WriteEnabled1=true
    Write Parameter Index 1=true
   )
   &SolveForcesAndVelocity(Clamp Velocity=false, Limit Acceleration=false)
   /SetVariables(Particles.Lightning glow color:LinearColor=&MakeLinearColorFromVectorAndFloat(
     "Vector (RGB)"=&MakeVectorFromLinearColorRGB(Color=&MultiplyLinearColor(
       Linear Color=<rgba 3 0.0462 2.0046 1>
       Scale Factor=/Niagara/DynamicInputs/Multiply/Multiply_Vector4_ByFloat(Vector4:Vector4f=<1 1 1 1>, Float:NiagaraFloat=1)
      ))
     "Float (Alpha)"=&Multiply_Float(
      A=&FloatFromCurve(FloatCurve=DataInterface() {propertyValues {Curve {keys [
           {arriveTangent=-0.6023, interpMode=RCIM_Cubic, leaveTangent=-0.6023, tangentMode=RCTM_User, time=0, value=0}
           <at 0.1 0.85 RCIM_Cubic>
           {arriveTangent=-1.052, interpMode=RCIM_Cubic, leaveTangent=-1.052, tangentMode=RCTM_User, time=0.85, value=0}
          ]}}})
      B=3
     )
    )) {isSetParameters=true}
   /SetVariables(
    Particles.Glow Dy02:Vector4f=&MakeVector4(
     X=&FloatFromCurve(FloatCurve=DataInterface() {propertyValues {Curve {keys [{tangentMode=RCTM_User, time=0, value=0}, {tangentMode=RCTM_User, time=1, value=1}]}}})
     Y=0.7
     Z=0
     W=2
    )
    Particles.Glow Dy03:Vector4f=&MakeVector4(
     X=&FN_FrontEndFloatModifier(Gameplay=1.5, Front End=0.7)
     Y=0
     Z=&Lerp_Float(A=1, B=1.5, Alpha=$Particles.Random)
     W=0
    )
   ) {isSetParameters=true}
   &FN_Sprite_Attributes(
    Camera Offset=10
    Color=&MakeLinearColorFromVectorAndFloat(
     "Vector (RGB)"=&Multiply_Vector(A=&MakeVectorFromLinearColorRGB(Color=<rgba 1.25 0 5 1>), B=<1 1 1>)
     "Float (Alpha)"=&Multiply_Float(
      A=&FloatFromCurve(FloatCurve=DataInterface() {propertyValues {Curve {keys [
           {interpMode=RCIM_Cubic, tangentMode=RCTM_User, time=0, value=0}
           {interpMode=RCIM_Cubic, tangentMode=RCTM_User, time=0.3, value=1}
           {arriveTangent=-1.4286, interpMode=RCIM_Cubic, leaveTangent=-1.4286, time=0.7, value=1}
           {tangentMode=RCTM_User, time=1, value=0}
          ]}}})
      B=1
     )
    )
    Create Camera Offset Attribute=true
    Create Color Attribute=true
    Create Position Attribute=true
    Create SpriteSize Attribute=true
    Position:NiagaraPosition=&TransformPosition(
     Position=&AddVectorToPosition(
      Position=$Particles.Position
      Vector=&RotateVectorByQuaternion(
       Vector To Rotate=&RandomRangeVector(Minimum=<20 0 0>, Maximum=<35 0 0>)
       Quaternion=$Particles.SkeletalMeshLocation.SampledRotation
      )
     )
     Source Space=::Local
     Destination Space=::Simulation
    )
    SpriteSize=&Multiply_Vector2D_ByFloat(
     Float=&FloatFromCurve(
      FloatCurve=DataInterface() {propertyValues {Curve {keys [
          {interpMode=RCIM_Cubic, tangentMode=RCTM_User, time=0, value=0}
          {interpMode=RCIM_Cubic, tangentMode=RCTM_User, time=0.3, value=1}
          {tangentMode=RCTM_User, time=1, value=0}
         ]}}}
      Scale Curve=1.2
     )
     Vector2D=<30 30>
    )
    Write to Module Namespace=true
   )
  ]
  rendererClasses [&NiagaraMeshRendererProperties]
  renderers [
   {rendererClass=&NiagaraMeshRendererProperties, rendererIndex=0}
   {rendererClass=&NiagaraMeshRendererProperties, rendererIndex=1}
  ]
 }
 SplitLightning_Center001 {
  emitterUpdateScript [
   &EmitterState(Life Cycle Mode=::NewEnumerator1, Loop Behavior=::NewEnumerator1, Loop Delay=$System.Delay, Loop Duration=2, UseLoopDelay=true)
   &SpawnRate(SpawnRate=50, Use Spawn Probability=true)
   &SpawnBurst_Instantaneous(Spawn Count=2, Spawn Time=0.005, Use Spawn Probability=true, UseLoopCountLimit=false)
  ]
  particleSpawnScript [
   &PartitionParticles(Number Of Partitions=1)
   &InitializeParticle(
    Color=&MultiplyLinearColor(
     Linear Color=&RandomRangeLinearColor(Minimum=<rgba 1 1 1 0.3>, Maximum=<rgba 1.2 1.2 1.2 1>)
     Scale Factor=&MakeVector4_FromFloat(Float=1)
    )
    Lifetime=&RandomRangeFloat(Minimum=0.2, Maximum=0.3)
    Mesh Scale Mode=::NewEnumerator3
    Mesh Uniform Scale=&RandomRangeFloat(Minimum=0.8, Maximum=1.3)
    UsePositionOffset=false
    Write Material Random=true
   )
   &SkeletalMeshLocation(
    Bone Sampling Mode=::NewEnumerator4
    Interpolate Mesh Sampling=false
    MaskSpawnGroup=false
    Orientation Sampling=::NewEnumerator1
    Sampled Position Offset=&AddVectorToPosition(Position=$System.FX position offset, Vector=&RandomRangeVector(Minimum=<0 -20 0>, Maximum=<0 -20 0>))
    Skeletal Mesh=$System.Skeletal Mesh
   )
   &ShapeLocation(MaskSpawnGroup=false, Non Uniform Scale=<1 1 0>, Offset Mode=::NewEnumerator0, Override Seed=false, Sphere Radius=25, UseOwnerScale=true)
   &InitialMeshOrientation(
    Mesh Axis To Orient=$Particles.Velocity
    Mesh Orientation Mode=::NewEnumerator2
    Orientation Vector=<0 0 -1>
    Rotation=<-0.25 0 0>
    Use Rotation Vector=false
   )
   &AddVelocity(
    Module.Use Falloff Distance=false
    Origin Offset=<0 -20 0>
    Override Seed=false
    Velocity Mode=::NewEnumerator1
    Velocity Origin=$Particles.SkeletalMeshLocation.SampledPosition
    Velocity Speed=400
   )
   &CameraOffset(Camera Offset Amount=-50)
  ]
  particleUpdateScript [
   &ParticleState()
   &ScaleMeshSize(Scale Factor=&RandomRangeVector(Minimum=<1 1 2>, Maximum=<1.5 1.8 2>)) {module=ScaleMeshSize001}
   &ScaleMeshSize(Scale Factor=&MakeVector(
     X=1
     Y=1
     Z=&FloatFromCurve(
      FloatCurve=DataInterface() {propertyValues {Curve {keys [
          {interpMode=RCIM_Cubic, leaveTangent=0.0966, tangentMode=RCTM_Break, time=0, value=1}
          {arriveTangent=0.0951, interpMode=RCIM_Cubic, leaveTangent=-0.1745, tangentMode=RCTM_Break, time=1, value=1.07}
         ]}}}
      Scale Curve=&FN_FrontEndFloatModifier(Gameplay=1, Front End=0.7)
     )
    ))
   &DynamicMaterialParameters(
    Index 0 Param 2=&RandomRangeFloat(Minimum=0, Maximum=2)
    Index 0 Param 4=&MakeCustomFloatFromBool(
     Bool=false
     True Float=11
     False Float=&SelectFloatFromArray(
      Array Sampling Mode=::NewEnumerator0
      Direct Array Index=?
      Float Selection Array=DataInterface() {propertyValues {FloatData [3,11,1,9,2,11,11]}}
      Random Seed=0
     )
    )
    Index 1 Param 1=&FloatFromCurve(FloatCurve=DataInterface() {propertyValues {Curve {keys [<at 0 1>,<at 1 1.3>]}}}, Scale Curve=0.5)
    Index 1 Param 2=2
    Index 1 Param 3=&FloatFromCurve(FloatCurve=DataInterface() {propertyValues {Curve {keys [{interpMode=RCIM_Cubic, tangentMode=RCTM_Break, time=0.5, value=-0.5}, <at 1 0.1>]}}})
    Index 2 Param 1=&Multiply_Float(
     A=&FloatFromCurve(
      FloatCurve=DataInterface() {propertyValues {Curve {keys [{interpMode=RCIM_Cubic, leaveTangent=8.1085, tangentMode=RCTM_Break, time=0.8568, value=1}, {tangentMode=RCTM_User, time=1, value=0}]}}}
      Scale Curve=&MakeCustomFloatFromBool(Bool=$User.bIsFrontEndPreview, True Float=0.3, False Float=&RandomRangeFloat(Minimum=0.5, Maximum=0.6))
     )
     B=&Divide_Float(
      A=&MakeFloatFromVector(Vector=$Particles.Scale, Channel=::NewEnumerator0)
      B=&RandomRangeFloat(Minimum=10, Maximum=15)
     )
    )
    Index 2 Param 2=&FloatFromCurve(
     FloatCurve=DataInterface() {propertyValues {Curve {keys [
         {interpMode=RCIM_Cubic, tangentMode=RCTM_Break, time=0.7, value=0}
         {arriveTangent=-0.0859, interpMode=RCIM_Cubic, tangentMode=RCTM_Break, time=1, value=-1}
        ]}}}
     Scale Curve=50
    )
    Index 2 Param 3=&Waveform(
     Global Input Position=&RandomRangeFloat(Minimum=0, Maximum=15)
     Global Phase=$Particles.Age
     "Waveform [1]"=::NewEnumerator12
     "[1] Phase Scale"=2
     "[1] Frequency"=1
     "[1] Amplitude Min/Max"=<0 1>
    )
    Index 2 Param 4=25
    Index 3 Param 1=&FloatFromCurve(
     FloatCurve=DataInterface() {propertyValues {Curve {keys [
         {interpMode=RCIM_Cubic, leaveTangent=-0.0025, tangentMode=RCTM_Break, time=0, value=0}
         {arriveTangent=1.8528, interpMode=RCIM_Cubic, tangentMode=RCTM_Break, time=1, value=1}
        ]}}}
     Scale Curve=&RandomRangeFloat(Minimum=15, Maximum=25)
    )
    Index 3 Param 2=&FloatFromCurve(FloatCurve=DataInterface() {propertyValues {Curve {keys [
         {interpMode=RCIM_Cubic, leaveTangent=-0.0025, tangentMode=RCTM_Break, time=0, value=0.5}
         {arriveTangent=-0.0316, interpMode=RCIM_Cubic, tangentMode=RCTM_Break, time=1, value=-0.2}
        ]}}})
    Index 3 Param 3=&FloatFromCurve(FloatCurve=DataInterface() {propertyValues {Curve {keys [
         {interpMode=RCIM_Cubic, leaveTangent=1.1034, tangentMode=RCTM_Break, time=0, value=-1}
         {arriveTangent=2.2399, tangentMode=RCTM_User, time=0.5, value=0.125}
        ]}}})
    Index 3 Param 4=&FloatFromCurve(FloatCurve=DataInterface() {propertyValues {Curve {keys [
         {interpMode=RCIM_Cubic, leaveTangent=3.0601, tangentMode=RCTM_Break, time=0.5, value=-0.5}
         {arriveTangent=3.2053, tangentMode=RCTM_User, time=1, value=1}
        ]}}})
    Module.Index0Param0Write=true
    Module.Index0Param1Write=true
    Module.Index0Param2Write=true
    Module.Index0Param3Write=true
    Module.Index1Param0Write=true
    Module.Index1Param1Write=true
    Module.Index1Param2Write=true
    Module.Index1Param3Write=true
    Module.Index2Param0Write=true
    Module.Index2Param1Write=true
    Module.Index2Param2Write=true
    Module.Index2Param3Write=true
    Module.Index3Param0Write=true
    Module.Index3Param1Write=true
    Module.Index3Param2Write=true
    Module.Index3Param3Write=true
    Write Parameter Index 1=true
    Write Parameter Index 2=true
    Write Parameter Index 3=true
   ) {module=DynamicMaterialParameters001}
   &Drag(Drag=2, Use Linear Drag=true, Use Rotational Drag=true)
   &SolveForcesAndVelocity(Clamp Velocity=false, Limit Acceleration=false)
   &ScaleColor(
    Scale Alpha=&FloatFromCurve(FloatCurve=DataInterface() {propertyValues {Curve {keys [
         {tangentMode=RCTM_User, time=0, value=1}
         {arriveTangent=-0.0405, interpMode=RCIM_Cubic, leaveTangent=-0.0405, tangentMode=RCTM_User, time=0.7009, value=1}
         {arriveTangent=-8.0089, interpMode=RCIM_Cubic, leaveTangent=-8.0089, tangentMode=RCTM_User, time=1, value=0}
        ]}}})
    Scale RGB=&Lerp_Vector(
     A=&VectorFromFloat(Value=&FloatFromCurve(FloatCurve=DataInterface() {propertyValues {Curve {keys [<at 0 0>,<at 0.3786 0.2>,<at 1 20>]}}}, Scale Curve=20))
     B=&VectorFromFloat(Value=&RandomRangeFloat(Minimum=1, Maximum=1, Evaluation Type=::NewEnumerator1))
     Alpha=&FloatFromCurve(FloatCurve=DataInterface() {propertyValues {Curve {keys [
          {interpMode=RCIM_Cubic, leaveTangent=2.5294, tangentMode=RCTM_Break, time=0, value=0}
          {arriveTangent=-0.2166, interpMode=RCIM_Cubic, tangentMode=RCTM_Break, time=0.4, value=1}
         ]}}})
    )
    ScaleA=true
    ScaleRGB=true
   ) {module=ScaleColor004}
   /SetVariables(Particles.MeshOrientation=&MultiplyQuaternion(
     Quaternion A=&MultiplyQuaternion(
      Quaternion A=&MakeQuatFromAxisAngle(Axis=<0 1 0>, AngleInDegrees=0)
      Quaternion B=&FindQuatBetweenTwoVectors(From Vector=<0 0 1>, To Vector=$Particles.Velocity)
     )
     Quaternion B=$Particles.SkeletalMeshLocation.SampledRotation
    )) {isSetParameters=true}
  ]
  rendererClasses [&NiagaraMeshRendererProperties]
  renderers [
   {rendererClass=&NiagaraMeshRendererProperties, rendererIndex=0}
   {rendererClass=&NiagaraMeshRendererProperties, rendererIndex=1}
  ]
 }
}
summary {
 systemName=NS_Glider_BentBaton_Deploy
 emitters [
  {emitterName=Lightning_Arc, rendererClasses [&NiagaraMeshRendererProperties]}
  {emitterName=FXMesh001, rendererClasses [&NiagaraMeshRendererProperties, &NiagaraLightRendererProperties]}
  {emitterName=Cloud001, rendererClasses [&NiagaraMeshRendererProperties]}
  {emitterName=SplitLightning_Center001, rendererClasses [&NiagaraMeshRendererProperties]}
 ]
 userVariables [
  User.bIsFrontEnd {type=NiagaraBool}
  User.bIsFrontEndPreview {type=NiagaraBool}
  User.ForwardDot {type=NiagaraFloat, defaultValue=1, description=""}
  User.RibbonIDOffset {type=NiagaraInt32}
  User.RightDot {type=NiagaraFloat, defaultValue=0, description=""}
  User.RotationMovement {type=NiagaraFloat, defaultValue=0, description=""}
 ]
}
