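# Idle Niagara effect for the BentBaton glider, attached to its skeletal mesh.
# Cloud and CloudRing emit partitioned mesh puffs from two offsets, around a persistent layered mesh with light and sprite, plus lightning arcs.
# System-level smoothed interpolations drive spawn probability, drag, scale and opacity.
# A front-end preview flag adjusts offsets, colors and intensity and enables side spikes, with time-of-day and platform scaling dimming some glows.
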NiagaraSystem /BRCosmetics/Gadgets/Assets/VinderTech_GliderChute/Glider_BentBaton/FX/NS_Glider_BentBaton_FlipAway_Idle
dependencies {
 usedDataInterfaces [
  &NiagaraDataInterfaceArrayFloat3
  &NiagaraDataInterfaceVectorCurve
  &NiagaraDataInterfaceCurve
  &NiagaraDataInterfaceColorCurve
  &NiagaraDataInterfaceArrayFloat
  &NiagaraDataInterfacePlatformSet
  &NiagaraDataInterfaceSkeletalMesh
 ]
 usedDynamicInputs [
  &Subtract_Float
  &Abs_Float
  &DotProduct
  /Niagara/DynamicInputs/Multiply/ScaleAndBiasFloat
  &Ceil_Float
 ]
 usedModules [
  &self:SystemSpawnScript.SystemScriptSource.SystemScriptGraph.NiagaraNodeAssignment_0.SetVariables_#1
  &self:SystemSpawnScript.SystemScriptSource.SystemScriptGraph.NiagaraNodeAssignment_1.SetVariables_#2
  &self:SystemSpawnScript.SystemScriptSource.SystemScriptGraph.NiagaraNodeAssignment_5.SetVariables_#3
  &self:SystemSpawnScript.SystemScriptSource.SystemScriptGraph.NiagaraNodeAssignment_2.SetVariables_#4
  &self:SystemSpawnScript.SystemScriptSource.SystemScriptGraph.NiagaraNodeAssignment_3.SetVariables_#5
  &SystemState
  &InterpolateOverTime
  &self:SystemSpawnScript.SystemScriptSource.SystemScriptGraph.NiagaraNodeAssignment_4.SetVariables_#6
 ]
}
emitters {
 Cloud {
  emitterUpdateScript [
   &EmitterState()
   &SpawnRate(SpawnRate=30, Use Spawn Probability=true)
   /SetVariables(Emitter.BaseEmissiveColor:LinearColor=&MakeCustomLinearColorFromBool(Bool=$User.bIsFrontEndPreview, True Color=<rgba 0.0361 0.0129 0.15 1>, False Color=<rgba 0 0 0 1>)) {isSetParameters=true}
  ]
  particleSpawnScript [
   &PartitionParticles()
   &InitializeParticle(
    Color=<rgba 0.8455 0.8455 0.8455 1>
    Lifetime Max=1
    Lifetime Min=0.7
    Lifetime Mode=::NewEnumerator1
    Mesh Scale=&Multiply_VectorByFloat(Vector=&RandomRangeVector(Minimum=<3 2 2>, Maximum=<4 3 3>), Float=0.5)
    Mesh Scale Mode=::NewEnumerator4
    UsePositionOffset=false
    Write Material Random=true
   )
   &SkeletalMeshLocation(
    Bone Sampling Mode=::NewEnumerator4
    Interpolate Mesh Sampling=false
    MaskSpawnGroup=false
    Orientation Sampling=::NewEnumerator1
    Sampled Position Offset=&FN_FrontEndVector3Modifier(
     Gameplay=&AddVectorToPosition(
      Position=&SelectVectorFromArray(
       Direct Array Index=$Particles.PartitionParticles.Partition
       Vector Selection Array=DataInterface() {propertyValues {FloatData [<-25 -20 -5>, <20 -10 -5>]}}
      )
      Vector=$System.FX position offset
     )
     Front End=&AddVectorToPosition(
      Position=&SelectVectorFromArray(
       Direct Array Index=$Particles.PartitionParticles.Partition
       Vector Selection Array=DataInterface() {propertyValues {FloatData [<-25 40 -10>, <20 40 -10>]}}
      )
      Vector=$System.FX position offset
     )
    )
    Skeletal Mesh=$System.Skeletal Mesh
   )
   &AddVelocity(
    Scale Added Velocity=&Multiply_VectorByFloat(
     Vector=&Multiply_VectorByFloat(
      Vector=<2.5 2.5 2.5>
      Float=&Lerp_Float(A=0.35, B=1, Alpha=$System.InterpolateOverTime004.InterpolateOverTime_Float.Moving Average)
     )
     Float=&Lerp_Float(A=1, B=0.6, Alpha=$System.InterpolateOverTime.InterpolateOverTime_Float.Moving Average)
    )
    Velocity=&SelectVectorFromArray(
     Direct Array Index=$Particles.PartitionParticles.Partition
     Vector Selection Array=DataInterface() {propertyValues {FloatData [<-25 120 0>, <25 120 0>]}}
    )
   )
   &AddVelocityInCone(Cone Angle=10, Cone Axis=<0 1 0>, Module.Use Velocity Falloff On Cone Axis=false, Velocity Strength=100) {module=AddVelocityInCone001}
   &InitialMeshOrientation(
    Mesh Orientation Mode=::NewEnumerator2
    Rotation=&RandomRangeVector(Minimum=<0 0 0>, Maximum=<1 0 0>)
    Use Rotation Vector=true
   )
   &CameraOffset()
  ]
  particleUpdateScript [
   &Drag(
    Drag=&Multiply_Float(
     A=&Multiply_Float(
      A=&MakeCustomFloatFromBool(
       Bool=&SetBoolByIntComparison(A=$Particles.PartitionParticles.Partition, B=0, Comparison Type=::NewEnumerator0)
       True Float=&Lerp_Float(A=0, B=1, Alpha=$System.InterpolateOverTime001.InterpolateOverTime_Float.Moving Average)
       False Float=&Lerp_Float(A=0, B=1, Alpha=$System.InterpolateOverTime002.InterpolateOverTime_Float.Moving Average)
      )
      B=1
     )
     B=1
    )
    Use Linear Drag=true
    Use Rotational Drag=true
   )
   &SolveForcesAndVelocity(Clamp Velocity=false, Limit Acceleration=false)
   &ParticleState()
   &ScaleMeshSize(Scale Factor=&Multiply_VectorByFloat(
     Vector=&Multiply_VectorByFloat(
      Vector=&Multiply_Vector(
       A=<1 1 1>
       B=&VectorFromCurve(VectorCurve=DataInterface() {propertyValues {XCurve {keys [<at 0 5>,<at 0.2 1>,<at 1 1>]}, YCurve {keys [<at 0 1>,<at 1 1>]}, ZCurve {keys [<at 0 1>,<at 1 1>]}}})
      )
      Float=&RandomRangeFloat(Minimum=1, Maximum=0.7)
     )
     Float=&FloatFromCurve(
      FloatCurve=DataInterface() {propertyValues {Curve {keys [
          {interpMode=RCIM_Cubic, tangentMode=RCTM_User, time=0, value=0.2}
          {arriveTangent=1.3433, tangentMode=RCTM_User, time=0.1866, value=0.3318}
          {arriveTangent=-1.0223, interpMode=RCIM_Cubic, leaveTangent=-1.0223, tangentMode=RCTM_User, time=1, value=0.8}
         ]}}}
      Scale Curve=&Multiply_Float(
       A=&Multiply_Float(
        A=&Multiply_Float(A=1.3, B=&Lerp_Float(A=0.7, B=1, Alpha=$System.InterpolateOverTime004.InterpolateOverTime_Float.Moving Average))
        B=&MakeCustomFloatFromBool(
         Bool=&SetBoolByIntComparison(A=$Particles.PartitionParticles.Partition, B=0, Comparison Type=::NewEnumerator0)
         True Float=&Lerp_Float(A=1, B=0.8, Alpha=$System.InterpolateOverTime001.InterpolateOverTime_Float.Moving Average)
         False Float=&Lerp_Float(A=1, B=0.8, Alpha=$System.InterpolateOverTime002.InterpolateOverTime_Float.Moving Average)
        )
       )
       B=&Lerp_Float(A=1, B=0.7, Alpha=$System.InterpolateOverTime.InterpolateOverTime_Float.Moving Average)
      )
     )
    ))
   &DynamicMaterialParameters(
    Index 0 Param 1=&FloatFromCurve(
     FloatCurve=DataInterface() {propertyValues {Curve {keys [
         {interpMode=RCIM_Cubic, tangentMode=RCTM_User, time=0, value=0}
         <at 0.2 0.55>
         {arriveTangent=-1.0223, interpMode=RCIM_Cubic, leaveTangent=-1.0223, tangentMode=RCTM_User, time=1, value=1}
        ]}}}
     Scale Curve=&RandomRangeFloat(Minimum=0.7, Maximum=1.5)
    )
    Index 0 Param 3=&Lerp_Float(A=0, B=1, Alpha=$System.InterpolateOverTime004.InterpolateOverTime_Float.Moving Average)
    Module.Index0Param0Write=true
    Module.Index0Param1Write=true
    Module.Index0Param2Write=true
    Module.Index0Param3Write=true
   )
   &ScaleColor(
    Scale Alpha=&FloatFromCurve(FloatCurve=DataInterface() {propertyValues {Curve {keys [
         {arriveTangent=2.8927, interpMode=RCIM_Cubic, leaveTangent=2.8927, tangentMode=RCTM_User, time=0, value=0}
         {arriveTangent=3.3333, interpMode=RCIM_Cubic, leaveTangent=3.3333, time=0.1, value=0.8}
         <at 0.3 1>
         <at 0.8 1>
         <at 1 0>
        ]}}})
    Scale RGB=&MakeVectorFromLinearColorRGB(Color=&ColorFromCurve(ColorCurve=DataInterface() {propertyValues {
        AlphaCurve {keys []}
        BlueCurve {keys [{arriveTangent=-331.6383, interpMode=RCIM_Cubic, time=0, value=5}, <at 0.3949 5>, <at 0.7598 3 RCIM_Cubic>]}
        GreenCurve {keys [{arriveTangent=-87.3695, interpMode=RCIM_Cubic, time=0, value=3.3333}, <at 0.3949 0>, <at 0.7598 0.6094 RCIM_Cubic>]}
        RedCurve {keys [<at 0 0 RCIM_Cubic>,<at 0.3949 1.0461>,<at 0.7598 0.6906 RCIM_Cubic>]}
       }}))
    ScaleA=true
    ScaleRGB=true
   )
   &FN_Mesh_Attributes(
    Color=&MakeLinearColorFromVectorAndFloat(
     "Vector (RGB)"=&Multiply_Vector(
      A=&MakeVectorFromLinearColorRGB(Color=<rgba 1 1 1 1>)
      B=&MakeVectorFromLinearColorRGB(Color=&ColorFromCurve(ColorCurve=DataInterface() {propertyValues {
          AlphaCurve {keys []}
          BlueCurve {keys [{arriveTangent=-331.6383, interpMode=RCIM_Cubic, time=0, value=5}, <at 0.3949 5>, <at 0.7598 3 RCIM_Cubic>]}
          GreenCurve {keys [{arriveTangent=-87.3695, interpMode=RCIM_Cubic, time=0, value=3.3333}, <at 0.3949 0>, <at 0.7598 0 RCIM_Cubic>]}
          RedCurve {keys [<at 0 0 RCIM_Cubic>,<at 0.3949 1.0461>,<at 0.7598 0.1019 RCIM_Cubic>]}
         }}))
     )
     "Float (Alpha)"=&Multiply_Float(
      A=&FloatFromCurve(FloatCurve=DataInterface() {propertyValues {Curve {keys [
           {arriveTangent=2.8927, interpMode=RCIM_Cubic, leaveTangent=2.8927, tangentMode=RCTM_User, time=0, value=0}
           {arriveTangent=3.3333, interpMode=RCIM_Cubic, leaveTangent=3.3333, time=0.1, value=1}
           <at 0.3 1>
           <at 0.8 1>
           <at 1 0>
          ]}}})
      B=1
     )
    )
    Create Color Attribute=true
    Create Scale Attribute=true
    Scale=&Multiply_Vector(
     A=<0.7 0.7 0.7>
     B=&Multiply_VectorByFloat(
      Vector=&Multiply_VectorByFloat(
       Vector=&Multiply_Vector(
        A=<1 1 1>
        B=&VectorFromCurve(
         VectorCurve=DataInterface() {propertyValues {XCurve {keys [<at 0 5>,<at 0.2 1>,<at 1 1>]}, YCurve {keys [<at 0 1>,<at 1 1>]}, ZCurve {keys [<at 0 1>,<at 1 1>]}}}
         Scale Curve=&Multiply_VectorByFloat(
          Vector=&Multiply_VectorByFloat(Vector=<1 1 1>, Float=1)
          Float=&FN_FrontEndFloatModifier(
           Gameplay=&Lerp_Float(
            A=0
            B=&MakeCustomFloatFromBool(Bool=&SetBoolByFloatComparison(A=$User.ForwardDot, B=0.05), True Float=1, False Float=0)
            Alpha=$System.InterpolateOverTime004.InterpolateOverTime_Float.Moving Average
           )
           Front End=1
          )
         )
        )
       )
       Float=&RandomRangeFloat(Minimum=1, Maximum=0.7)
      )
      Float=&FloatFromCurve(
       FloatCurve=DataInterface() {propertyValues {Curve {keys [
           {interpMode=RCIM_Cubic, tangentMode=RCTM_User, time=0, value=0.2}
           {arriveTangent=1.3433, tangentMode=RCTM_User, time=0.2, value=0.3}
           <at 0.8 0.7>
           {arriveTangent=-1.0223, interpMode=RCIM_Cubic, leaveTangent=-1.0223, tangentMode=RCTM_User, time=1, value=0}
          ]}}}
       Scale Curve=&Multiply_Float(
        A=&Multiply_Float(
         A=1.5
         B=&MakeCustomFloatFromBool(
          Bool=&SetBoolByIntComparison(A=$Particles.PartitionParticles.Partition, B=0, Comparison Type=::NewEnumerator0)
          True Float=&Lerp_Float(A=1, B=0.8, Alpha=$System.InterpolateOverTime001.InterpolateOverTime_Float.Moving Average)
          False Float=&Lerp_Float(A=1, B=0.8, Alpha=$System.InterpolateOverTime002.InterpolateOverTime_Float.Moving Average)
         )
        )
        B=&Lerp_Float(A=1, B=0.7, Alpha=$System.InterpolateOverTime.InterpolateOverTime_Float.Moving Average)
       )
      )
     )
    )
    Write to Module Namespace=true
   )
   /SetVariables(Particles.CameraOffset02:NiagaraFloat=&FN_FrontEndFloatModifier(Gameplay=-50, Front End=-20)) {isSetParameters=true}
  ]
  rendererClasses [&NiagaraMeshRendererProperties]
  renderers [
   {rendererClass=&NiagaraMeshRendererProperties, rendererIndex=0}
   {rendererClass=&NiagaraMeshRendererProperties, rendererIndex=1}
  ]
 }
 CloudRing {
  emitterUpdateScript [
   &EmitterState()
   &SpawnRate(
    Spawn Probability=&Lerp_Float(A=0, B=1, Alpha=$System.InterpolateOverTime004.InterpolateOverTime_Float.Moving Average)
    SpawnRate=42
    Use Spawn Probability=true
   )
   /SetVariables(Emitter.BaseEmissiveColorOutside:LinearColor=&MakeCustomLinearColorFromBool(Bool=$User.bIsFrontEndPreview, True Color=<rgba 0.0318 0.0316 0.2 1>, False Color=<rgba 0 0 0 1>)) {isSetParameters=true}
   /SetVariables(Emitter.BaseEmissiveColorInSide:LinearColor=&MakeCustomLinearColorFromBool(Bool=$User.bIsFrontEndPreview, True Color=<rgba 0 0.1165 0.2 1>, False Color=<rgba 0 0 0 1>)) {isSetParameters=true}
  ]
  particleSpawnScript [
   &PartitionParticles()
   &InitializeParticle(
    Color=&MakeCustomLinearColorFromBool(Bool=&RandomBool(), True Color=<rgba 0.3245 1.9618 2 1>, False Color=<rgba 1 1 1 1>)
    Lifetime Max=1
    Lifetime Min=0.9
    Lifetime Mode=::NewEnumerator1
    Mesh Scale=&Multiply_VectorByFloat(
     Vector=&RandomRangeVector(Minimum=<2.5 2.5 2.5>, Maximum=<4 4 4>)
     Float=&FN_FrontEndFloatModifier(Gameplay=0.39, Front End=0.35)
    )
    Mesh Scale Mode=::NewEnumerator4
    UsePositionOffset=false
    Write Material Random=true
   )
   &SkeletalMeshLocation(
    Bone Sampling Mode=::NewEnumerator4
    Interpolate Mesh Sampling=false
    MaskSpawnGroup=false
    Orientation Sampling=::NewEnumerator1
    Sampled Position Offset=&AddVectorToPosition(
     Position=&AddVectorToPosition(
      Position=&SelectVectorFromArray(
       Direct Array Index=$Particles.PartitionParticles.Partition
       Vector Selection Array=DataInterface() {propertyValues {FloatData [<-25 -20 -7>, <20 0 -7>]}}
      )
      Vector=$System.FX position offset
     )
     Vector=<0 40 0>
    )
    Skeletal Mesh=$System.Skeletal Mesh
   )
   &AddVelocity(
    Scale Added Velocity=&Multiply_VectorByFloat(
     Vector=&Multiply_VectorByFloat(Vector=<2 2.3 2>, Float=1)
     Float=&Lerp_Float(A=1, B=0.6, Alpha=$System.InterpolateOverTime.InterpolateOverTime_Float.Moving Average)
    )
    Velocity=&SelectVectorFromArray(
     Direct Array Index=$Particles.PartitionParticles.Partition
     Vector Selection Array=DataInterface() {propertyValues {FloatData [<-30 120 0>, <30 120 0>]}}
    )
   )
   &AddVelocityInCone(Cone Angle=30, Cone Axis=<0 1 0>, Module.Use Velocity Falloff On Cone Axis=false, Velocity Strength=100) {module=AddVelocityInCone001}
   &InitialMeshOrientation(Mesh Orientation Mode=::NewEnumerator2, Rotation=<0.25 0 0>, Use Rotation Vector=true)
   &CameraOffset(Camera Offset Amount=&FN_FrontEndFloatModifier(Gameplay=-80, Front End=0))
   &InitialMeshOrientation(
    Mesh Orientation Mode=::NewEnumerator2
    Rotation=&SelectVectorFromArray(
     Direct Array Index=$Particles.PartitionParticles.Partition
     Vector Selection Array=DataInterface() {propertyValues {FloatData [<0 0 0>, <0 0.5 0>]}}
    )
    Use Rotation Vector=true
   ) {module=InitialMeshOrientation002}
   &InitialMeshOrientation(
    Mesh Orientation Mode=::NewEnumerator2
    Rotation=&RandomRangeVector(Minimum=<0 -0.1 0>, Maximum=<0 0.1 0>)
    Use Rotation Vector=true
   ) {module=InitialMeshOrientation003}
  ]
  particleUpdateScript [
   &Drag(
    Drag=&Multiply_Float(
     A=&MakeCustomFloatFromBool(
      Bool=&SetBoolByIntComparison(A=$Particles.PartitionParticles.Partition, B=0, Comparison Type=::NewEnumerator0)
      True Float=&Lerp_Float(A=0, B=1, Alpha=$System.InterpolateOverTime001.InterpolateOverTime_Float.Moving Average)
      False Float=&Lerp_Float(A=0, B=1, Alpha=$System.InterpolateOverTime002.InterpolateOverTime_Float.Moving Average)
     )
     B=1
    )
    Use Linear Drag=true
    Use Rotational Drag=true
   )
   &SolveForcesAndVelocity(Clamp Velocity=false, Limit Acceleration=false)
   &ParticleState()
   &UpdateMeshOrientation(
    Rotation Rate=&RandomRangeFloat(
     Minimum=&SelectFloatFromArray(Direct Array Index=$Particles.PartitionParticles.Partition, Float Selection Array=DataInterface() {propertyValues {FloatData [1.5,-1.5]}})
     Maximum=&SelectFloatFromArray(Direct Array Index=$Particles.PartitionParticles.Partition, Float Selection Array=DataInterface() {propertyValues {FloatData [2.5,-2.5]}})
    )
    Rotation Vector=&RandomRangeVector(Minimum=<0 0 1>, Maximum=<0 0 1>)
   )
   &ScaleMeshSize(Scale Factor=&Multiply_VectorByFloat(
     Vector=&Multiply_VectorByFloat(
      Vector=&MakeCustomVectorFromBool(Bool=&RandomBool(), True Vector=<1 1 1>, False Vector=<1 1 -1>)
      Float=&RandomRangeFloat(Minimum=1, Maximum=1.3)
     )
     Float=&FloatFromCurve(
      FloatCurve=DataInterface() {propertyValues {Curve {keys [
          {arriveTangent=0.176, interpMode=RCIM_Cubic, leaveTangent=0.176, tangentMode=RCTM_User, time=-0.0015, value=0.3}
          {arriveTangent=0.1175, interpMode=RCIM_Cubic, leaveTangent=0.1175, time=0.2, value=0.345}
          {arriveTangent=0.0563, interpMode=RCIM_Cubic, leaveTangent=0.0563, tangentMode=RCTM_User, time=0.6289, value=0.3741}
          {arriveTangent=0.098, interpMode=RCIM_Cubic, leaveTangent=0.098, tangentMode=RCTM_User, time=1, value=0.385}
         ]}}}
      Scale Curve=&Multiply_Float(
       A=&Multiply_Float(
        A=1.5
        B=&MakeCustomFloatFromBool(
         Bool=&SetBoolByIntComparison(A=$Particles.PartitionParticles.Partition, B=0, Comparison Type=::NewEnumerator0)
         True Float=&Lerp_Float(A=1, B=0.7, Alpha=$System.InterpolateOverTime001.InterpolateOverTime_Float.Moving Average)
         False Float=&Lerp_Float(A=1, B=0.7, Alpha=$System.InterpolateOverTime002.InterpolateOverTime_Float.Moving Average)
        )
       )
       B=&Lerp_Float(A=1, B=0.7, Alpha=$System.InterpolateOverTime.InterpolateOverTime_Float.Moving Average)
      )
     )
    ))
   &DynamicMaterialParameters(
    Index 0 Param 1=&FloatFromCurve(
     FloatCurve=DataInterface() {propertyValues {Curve {keys [
         {interpMode=RCIM_Cubic, tangentMode=RCTM_User, time=0, value=0}
         <at 0.2 0.55>
         {arriveTangent=-1.0223, interpMode=RCIM_Cubic, leaveTangent=-1.0223, tangentMode=RCTM_User, time=1, value=1}
        ]}}}
     Scale Curve=&RandomRangeFloat(Minimum=0.7, Maximum=1.5)
    )
    Index 0 Param 3=&Lerp_Float(A=0, B=1, Alpha=$System.InterpolateOverTime004.InterpolateOverTime_Float.Moving Average)
    Index 0 Param 4=&Multiply_Float(
     A=&FloatFromCurve(FloatCurve=DataInterface() {propertyValues {Curve {keys [
          {arriveTangent=2.2076, interpMode=RCIM_Cubic, leaveTangent=2.2076, tangentMode=RCTM_User, time=0, value=0.2}
          {arriveTangent=1.6957, interpMode=RCIM_Cubic, leaveTangent=1.6957, tangentMode=RCTM_User, time=1, value=1}
         ]}}})
     B=3
    )
    Module.Index0Param0Write=true
    Module.Index0Param1Write=true
    Module.Index0Param2Write=true
    Module.Index0Param3Write=true
   )
   &ScaleColor(
    Scale Alpha=&FloatFromCurve(FloatCurve=DataInterface() {propertyValues {Curve {keys [
         {arriveTangent=1.0996, interpMode=RCIM_Cubic, leaveTangent=1.0996, tangentMode=RCTM_User, time=0, value=0}
         <at 0.3 1>
         <at 0.6 0.3>
         <at 0.7 0>
        ]}}})
    Scale RGB=&MakeVectorFromLinearColorRGB(Color=&ColorFromCurve(ColorCurve=DataInterface() {propertyValues {
        AlphaCurve {keys []}
        BlueCurve {keys [<at 0.0418 30>,<at 0.8287 0.4896>]}
        GreenCurve {keys [<at 0.0418 1.3096>,<at 0.8287 0.1563>]}
        RedCurve {keys [<at 0.0418 12.4965>,<at 0.8287 0.337>]}
       }}))
    ScaleA=true
    ScaleRGB=true
   )
   &FN_Mesh_Attributes(
    Color=&MakeLinearColorFromVectorAndFloat(
     "Vector (RGB)"=&Multiply_Vector(A=&MakeVectorFromLinearColorRGB(Color=<rgba 0 1.3333 2 1>), B=<1 1 1>)
     "Float (Alpha)"=&Multiply_Float(
      A=1
      B=&FloatFromCurve(FloatCurve=DataInterface() {propertyValues {Curve {keys [
           {arriveTangent=2.8927, interpMode=RCIM_Cubic, leaveTangent=2.8927, tangentMode=RCTM_User, time=0, value=0}
           <at 0.3 1>
           <at 0.6 0.3>
           <at 0.7 0>
          ]}}})
     )
    )
    Create Color Attribute=true
    Write to Module Namespace=true
   )
  ]
  rendererClasses [&NiagaraMeshRendererProperties]
  renderers [
   {rendererClass=&NiagaraMeshRendererProperties, rendererIndex=0}
   {rendererClass=&NiagaraMeshRendererProperties, rendererIndex=1}
  ]
 }
 FXMesh {
  emitterUpdateScript [
   &EmitterState(Life Cycle Mode=::NewEnumerator1, Loop Behavior=::NewEnumerator1, Loop Duration Mode=::NewEnumerator1, UseLoopDelay=false)
   &SpawnBurst_Instantaneous(Use Spawn Probability=false, UseLoopCountLimit=false)
  ]
  particleSpawnScript [
   &InitializeParticle(
    Color Mode=::NewEnumerator0
    Mesh Scale Mode=::NewEnumerator3
    Mesh Uniform Scale=3
    Position Offset=<0 0 -10>
    UsePositionOffset=false
    Write Material Random=true
   )
   &InitialMeshOrientation(Mesh Orientation Mode=::NewEnumerator2, Rotation=<0 0.25 0>, Use Rotation Vector=true)
  ]
  particleUpdateScript [
   &SkeletalMeshLocation(
    Bone Sampling Mode=::NewEnumerator4
    MaskSpawnGroup=false
    Sampled Position Offset=&AddVectorToPosition(Position=Linked() {linkedVariable {name=System.FXMeshPosition, type=Vector3f}}, Vector=<0 0 0>)
    Skeletal Mesh=$System.Skeletal Mesh
   )
   &ParticleState(Kill Particles When Lifetime Has Elapsed=false, Let Infinitely Lived Particles Die When Emitter Deactivates=true)
   &ScaleMeshSize(Scale Factor=&Multiply_VectorByFloat(Vector=<1 1.5 1.3>, Float=$System.SizeScale))
   &FN_Mesh_Attributes(
    Color=&MakeLinearColorFromVectorAndFloat(
     "Vector (RGB)"=&MakeVectorFromLinearColorRGB(Color=<rgba 0.552 0 1 0.502>)
     "Float (Alpha)"=&Multiply_Float(
      A=&FNTimeOfDayScalar(Night Scale=0.2, Morning Scale=0.4, Day Scale=0.5, Evening Scale=0.4)
      B=&FN_FrontEndFloatModifier(Gameplay=1, Front End=1.2)
     )
    )
    Create Color Attribute=true
    Create Scale Attribute=true
    Scale=&Multiply_VectorByFloat(
     Vector=&Multiply_Vector(A=<1 1.5 1>, B=&VectorFromFloat(Value=$System.SizeScale))
     Float=4
    )
    Write to Module Namespace=true
   )
   &FN_Mesh_Attributes(
    Color=&MakeLinearColorFromVectorAndFloat("Vector (RGB)"=&MakeVectorFromLinearColorRGB(Color=<rgba 0.1093 0.0045 0.2 1>), "Float (Alpha)"=1)
    Create Color Attribute=true
    Create Scale Attribute=true
    Scale=&Multiply_VectorByFloat(
     Vector=&Multiply_Vector(A=<1.7 1.5 1>, B=&VectorFromFloat(Value=$System.SizeScale))
     Float=0.4
    )
    Write to Module Namespace=true
   ) {module=FN_Mesh_Attributes001}
   &FN_Mesh_Attributes(
    Color=&MakeLinearColorFromVectorAndFloat(
     "Vector (RGB)"=&MakeVectorFromLinearColorRGB(Color=<rgba 0 0 0 1>)
     "Float (Alpha)"=&FN_FrontEndFloatModifier(Gameplay=1, Front End=1)
    )
    Create Color Attribute=true
    Create Scale Attribute=true
    Scale=&Multiply_VectorByFloat(
     Vector=&Multiply_Vector(A=<1.7 1.5 1>, B=&VectorFromFloat(Value=$System.SizeScale))
     Float=0.2
    )
    Write to Module Namespace=true
   ) {module=FN_Mesh_Attributes002}
   &FN_Mesh_Attributes(
    Color=&MakeLinearColorFromVectorAndFloat(
     "Vector (RGB)"=&MakeVectorFromLinearColorRGB(Color=<rgba 0.3281 0 0.2137 1>)
     "Float (Alpha)"=&FN_FrontEndFloatModifier(Gameplay=1, Front End=1)
    )
    Create Color Attribute=true
    Create Scale Attribute=true
    Scale=&Multiply_VectorByFloat(
     Vector=&Multiply_Vector(A=<1 1.5 1.3>, B=&VectorFromFloat(Value=$System.SizeScale))
     Float=2.3
    )
    Write to Module Namespace=true
   ) {module=FN_Mesh_Attributes003}
   &FN_Mesh_Attributes(
    Color=&MakeLinearColorFromVectorAndFloat(
     "Vector (RGB)"=&Multiply_VectorByFloat(
      Vector=&MakeVectorFromLinearColorRGB(Color=<rgba 0 1.7337 2 1>)
      Float=&FN_FrontEndFloatModifier(Gameplay=1, Front End=1)
     )
     "Float (Alpha)"=3
    )
    Create Color Attribute=true
    Create Scale Attribute=true
    Scale=&Multiply_VectorByFloat(
     Vector=&Multiply_Vector(A=<1 1.5 1.3>, B=&VectorFromFloat(Value=$System.SizeScale))
     Float=2.9
    )
    Write to Module Namespace=true
   ) {module=FN_Mesh_Attributes004}
   &Light_Attributes(
    Color=1
    Color Mode=0
    Falloff Exponent=2
    Light Color:LinearColor=&Multiply_LinearColorByFloat(Linear Color=<rgba 29.8229 2.2782 30 1>, Float=1.5)
    Light Exponent=3
    Light Position Offset:Vector3f=<10 100 0>
    Light Radius=150
    Position=1
    Position Offset Coordinate Space:ENiagaraCoordinateSpace=::Local
    Radius=2
    UsePositionOffset=true
   )
   &FN_Sprite_Attributes(
    Camera Offset=50
    Color=&MakeLinearColorFromVectorAndFloat(
     "Vector (RGB)"=&MakeVectorFromLinearColorRGB(Color=<rgba 3 0.0156 2.0989 1>)
     "Float (Alpha)"=&Multiply_Float(
      A=1
      B=&FloatFromCurve(
       FloatCurve=DataInterface() {propertyValues {Curve {keys [<at 0 0>,<at 0.2 1>,<at 0.8 1>,<at 1 1>]}}}
       Scale Curve=&RandomRangeFloat(Minimum=0.5, Maximum=1, Evaluation Type=::NewEnumerator1)
      )
     )
    )
    Create Camera Offset Attribute=true
    Create Color Attribute=true
    Create Position Attribute=true
    Create SpriteSize Attribute=true
    Position:NiagaraPosition=&AddVectorToPosition(
     Position=$Particles.Position
     Vector=&RotateVectorByQuaternion(Vector To Rotate=<0 50 0>, Quaternion=$Particles.SkeletalMeshLocation.SampledRotation)
    )
    SpriteSize=&Vector2DFromFloat(Value=150)
    Write to Module Namespace=true
   )
  ]
  rendererClasses [
   &NiagaraMeshRendererProperties
   &NiagaraLightRendererProperties
   &NiagaraSpriteRendererProperties
  ]
  renderers [
   {rendererClass=&NiagaraMeshRendererProperties, rendererIndex=0}
   {rendererClass=&NiagaraMeshRendererProperties, rendererIndex=1}
   {rendererClass=&NiagaraMeshRendererProperties, rendererIndex=2}
   {rendererClass=&NiagaraMeshRendererProperties, rendererIndex=3}
   {rendererClass=&NiagaraLightRendererProperties, rendererIndex=4}
   {rendererClass=&NiagaraMeshRendererProperties, rendererIndex=5}
   {rendererClass=&NiagaraMeshRendererProperties, rendererIndex=6}
   {rendererClass=&NiagaraSpriteRendererProperties, rendererIndex=7}
  ]
 }
 Frontend_SideSpikes {
  emitterUpdateScript [
   /SetVariables(Emitter.RandomFloat:NiagaraFloat=&RandomRangeFloat(Minimum=0.3, Maximum=1)) {isSetParameters=true, module=SetVariables_#7}
   &EmitterState(
    Life Cycle Mode=::NewEnumerator1
    Loop Behavior=::NewEnumerator1
    Loop Delay=$System.Delay
    Loop Duration Mode=::NewEnumerator1
    UseLoopDelay=true
   )
   &SpawnBurst_Instantaneous(
    Spawn Probability=&MakeCustomFloatFromBool(Bool=$User.bIsFrontEndPreview, True Float=1, False Float=0)
    Use Spawn Probability=true
    UseLoopCountLimit=false
   )
  ]
  particleSpawnScript [
   &InitializeParticle(
    Color=<rgba 0.0781 0.9647 3 2>
    Mesh Scale=<1.6 2.1 2.2>
    Mesh Scale Mode=::NewEnumerator4
    UsePositionOffset=false
    Write Material Random=true
   )
   &InitialMeshOrientation(Mesh Orientation Mode=::NewEnumerator2, Rotation=<0 0 0.25>, Use Rotation Vector=true)
   &InitialMeshOrientation(Mesh Orientation Mode=::NewEnumerator2, Rotation=<0 0.25 0>, Use Rotation Vector=true) {module=InitialMeshOrientation001}
   &CameraOffset(Camera Offset Amount=-10)
   &InitialMeshOrientation(Mesh Orientation Mode=::NewEnumerator2, Rotation=<-0.002 0 -0.01>, Use Rotation Vector=true) {module=InitialMeshOrientation002}
  ]
  particleUpdateScript [
   /SetVariables(Particles.FlashColor:LinearColor=&LerpLinearColors(StartColor=<rgba 1 1 1 1>, EndColor=<rgba 0 0.0722 0.401 1>, LerpFactor=0)) {isSetParameters=true, module=SetVariables_#8}
   &SkeletalMeshLocation(Bone Sampling Mode=::NewEnumerator4, MaskSpawnGroup=false, Sampled Position Offset=<-5 -25 -12>, Skeletal Mesh=$System.Skeletal Mesh)
   &ParticleState(Kill Particles When Lifetime Has Elapsed=false, Let Infinitely Lived Particles Die When Emitter Deactivates=true)
   &ScaleMeshSize(Scale Factor=&Multiply_VectorByFloat(Vector=<1 0.45 1>, Float=&Multiply_Float(A=0.8, B=$System.SizeScale)))
   &ScaleColor(
    Scale Alpha=&Multiply_Float(A=&Lerp_Float(A=0, B=1, Alpha=$System.InterpolateOverTime004.InterpolateOverTime_Float.Moving Average), B=1)
    Scale RGB=&Multiply_VectorByFloat(
     Vector=<1 1 1>
     Float=&Multiply_Float(A=1, B=&FN_PickFloatDependOnPlatform(TargetPlatformActive=1, TargetPlatformNoActive=0.6))
    )
    ScaleA=true
    ScaleRGB=true
   ) {enabled=false}
   &FN_Sprite_Attributes(
    Color=&MakeLinearColorFromVectorAndFloat(
     "Vector (RGB)"=&MakeVectorFromLinearColorRGB(Color=&MultiplyLinearColor(
       Linear Color=<rgba 0 0.8239 1 2>
       Scale Factor=&MakeVector4_FromFloat(Float=&Multiply_Float(A=1, B=&FN_PickFloatDependOnPlatform(TargetPlatformActive=1, TargetPlatformNoActive=0.6)))
      ))
     "Float (Alpha)"=&Multiply_Float(
      A=&Lerp_Float(
       A=0
       B=&FN_FrontEndFloatModifier(Gameplay=2, Front End=0.5)
       Alpha=$System.InterpolateOverTime004.InterpolateOverTime_Float.Moving Average
      )
      B=&FloatFromCurve(
       FloatCurve=DataInterface() {propertyValues {Curve {defaultValue=1, keys [{tangentMode=RCTM_User, time=0.5, value=0}, {tangentMode=RCTM_User, time=3, value=1}]}}}
       CurveIndex=$Emitter.Age
      )
     )
    )
    Create Color Attribute=true
    Write to Module Namespace=true
   ) {module=FN_Sprite_Attributes001}
   &DynamicMaterialParameters(
    Index 0 Param 2=1
    Index 0 Param 3=&Add_Float(
     A=&FloatFromCurve(
      FloatCurve=DataInterface() {propertyValues {Curve {keys [
          {arriveTangent=0.5715, interpMode=RCIM_Cubic, leaveTangent=0.5715, tangentMode=RCTM_User, time=0, value=0}
          {arriveTangent=1.1179, interpMode=RCIM_Cubic, leaveTangent=1.1179, tangentMode=RCTM_User, time=0.5, value=0.6}
          {arriveTangent=0.5282, interpMode=RCIM_Cubic, leaveTangent=0.5282, tangentMode=RCTM_User, time=1, value=1}
         ]}}}
      Scale Curve=0.95
     )
     B=1
    )
    Index 0 Param 4=&UniformAOrBFloat(A=-1, Evaluation Type=::NewEnumerator0, Recalculate Random Each Loop=true, Override Seed=false)
    Index 3 Param 1=0.3
    Module.Index0Param0Write=true
    Module.Index0Param1Write=true
    Module.Index0Param2Write=true
    Module.Index0Param3Write=true
    Module.Index3Param0Write=true
    Module.Index3Param1Write=false
    Module.Index3Param2Write=false
    Module.Index3Param3Write=false
    Write Parameter Index 3=true
   )
   /SetVariables(Particles.CenterPosition:NiagaraPosition=&AddVectorToPosition(Position=&AddVectorToPosition(Position=$Particles.Position, Vector=<-12 0 13>), Vector=<0 0 0>)) {isSetParameters=true, module=SetVariables_#9}
   &FN_Mesh_Attributes(
    Color=&MakeLinearColorFromVectorAndFloat(
     "Vector (RGB)"=&MakeVectorFromLinearColorRGB(Color=&MultiplyLinearColor(
       Linear Color=<rgba 0 0.849 1 1>
       Scale Factor=&MakeVector4_FromLinearColor(LinearColor=/Niagara/DynamicInputs/TypeConversions/MakeLinearColorFromFloat(Value:NiagaraFloat=&Multiply_Float(A=0.5, B=&Lerp_Float(A=1, B=1, Alpha=$System.InterpolateOverTime004.InterpolateOverTime_Float.Moving Average))))
      ))
     "Float (Alpha)"=&FloatFromCurve(
      FloatCurve=DataInterface() {propertyValues {Curve {defaultValue=1, keys [{tangentMode=RCTM_User, time=0.5, value=0}, {tangentMode=RCTM_User, time=3, value=1}]}}}
      CurveIndex=$Emitter.Age
      Scale Curve=&FN_FrontEndFloatModifier(Gameplay=1, Front End=0.5)
     )
    )
    Create Color Attribute=true
    Create Scale Attribute=true
    Scale=&Multiply_VectorByFloat(Vector=<1.5 1 1>, Float=&Multiply_Float(A=1.1, B=$System.SizeScale))
    Write to Module Namespace=true
   )
   &FN_Mesh_Attributes(
    Color=&MakeLinearColorFromVectorAndFloat(
     "Vector (RGB)"=&MakeVectorFromLinearColorRGB(Color=&MultiplyLinearColor(
       Linear Color=<rgba 0.3546 0 1 1>
       Scale Factor=&MakeVector4_FromFloat(Float=&Multiply_Float(
         A=&FN_FrontEndFloatModifier(Gameplay=1, Front End=2)
         B=&FN_PickFloatDependOnPlatform(TargetPlatformActive=1, TargetPlatformNoActive=0.6)
        ))
      ))
     "Float (Alpha)"=&Multiply_Float(
      A=&Lerp_Float(A=0, B=1, Alpha=$System.InterpolateOverTime004.InterpolateOverTime_Float.Moving Average)
      B=&FloatFromCurve(
       FloatCurve=DataInterface() {propertyValues {Curve {defaultValue=1, keys [{tangentMode=RCTM_User, time=0.5, value=0}, {tangentMode=RCTM_User, time=3, value=1}]}}}
       CurveIndex=$Emitter.Age
      )
     )
    )
    Create Color Attribute=true
    Create Scale Attribute=true
    Scale=&Multiply_Vector(
     A=<1.6 1 2.2>
     B=&Multiply_VectorByFloat(Vector=<0.8 1 1>, Float=&Multiply_Float(A=0.7, B=$System.SizeScale))
    )
    Write to Module Namespace=true
   ) {module=FN_Mesh_Attributes001}
  ]
  rendererClasses [&NiagaraMeshRendererProperties]
  renderers [
   {rendererClass=&NiagaraMeshRendererProperties, rendererIndex=0}
   {rendererClass=&NiagaraMeshRendererProperties, rendererIndex=1}
   {rendererClass=&NiagaraMeshRendererProperties, rendererIndex=2}
  ]
 }
 LightningBlur001 {
  emitterUpdateScript [
   &EmitterState(Life Cycle Mode=::NewEnumerator1, Loop Duration=0.7, Recalculate Duration Each Loop=true, UseLoopDelay=true)
   &SpawnRate(Spawn Probability=$System.InterpolateOverTime004.InterpolateOverTime_Float.Moving Average, SpawnRate=10, Use Spawn Probability=true)
  ]
  particleSpawnScript [
   &PartitionParticles()
   &InitializeParticle(
    Color=<rgba 0.0055 1.4166 3 1>
    Lifetime Max=0.3
    Lifetime Min=0.2
    Lifetime Mode=::NewEnumerator1
    Mesh Scale=&RandomRangeVector(Minimum=<1 1 1.5>, Maximum=<1 1 3>)
    Mesh Scale Mode=::NewEnumerator4
    UsePositionOffset=false
    Write Material Random=true
   )
   &CameraOffset()
   &ShapeLocation(
    Handle Radius=0
    Large Radius=20
    MaskSpawnGroup=false
    Offset Mode=::NewEnumerator0
    Override Seed=false
    Shape Primitive=::NewEnumerator3
    UseOwnerScale=true
   )
   &InitialMeshOrientation(
    Mesh Orientation Mode=::NewEnumerator2
    Rotation=&RandomRangeVector(Minimum=<-0.25 0 0>, Maximum=<-0.25 0 0>)
    Use Rotation Vector=true
   ) {module=InitialMeshOrientation002}
   &AddVelocity(Module.Use Falloff Distance=false, Origin Offset=<0 0 100>, Override Seed=false, Velocity Mode=::NewEnumerator1, Velocity Speed=1)
   &InitialMeshOrientation(
    Mesh Orientation Mode=::NewEnumerator2
    Rotation=&RandomRangeVector(Minimum=<0 0 0>, Maximum=<0 1 0>)
    Use Rotation Vector=true
   )
   /SetVariables(Particles.Random=&RandomRangeFloat(Minimum=0, Maximum=1)) {isSetParameters=true}
   &InitialMeshOrientation(
    Mesh Orientation Mode=::NewEnumerator2
    Rotation=&RandomRangeVector(Minimum=<0 0 -0.14>, Maximum=<0 0 0.14>)
    Use Rotation Vector=true
   ) {module=InitialMeshOrientation001}
  ]
  particleUpdateScript [
   &SkeletalMeshLocation(
    Bone Sampling Mode=::NewEnumerator4
    MaskSpawnGroup=false
    Sampled Position Offset=&AddVectorToPosition(Position=$System.FXMeshPosition, Vector=<0 50 0>)
    Skeletal Mesh=$System.Skeletal Mesh
   )
   &ParticleState()
   &ScaleMeshSize(Scale Factor=&Multiply_VectorByFloat(Vector=<1 1 1>, Float=0.4))
   &ScaleColor(
    Scale Alpha=&Multiply_Float(
     A=1
     B=&FloatFromCurve(FloatCurve=DataInterface() {propertyValues {Curve {keys [<at 0 1>,<at 1 0>]}}}, Scale Curve=2)
    )
    Scale RGB=&Multiply_VectorByFloat(
     Vector=&Multiply_VectorByFloat(Vector=<1 1 1>, Float=1)
     Float=&Multiply_Float(
      A=&RandomRangeFloat(Minimum=0, Maximum=1, Evaluation Type=::NewEnumerator1)
      B=&Multiply_Float(
       A=1
       B=&MakeCustomFloatFromBool(Bool=&RandomBool(), True Float=0.3, False Float=2)
      )
     )
    )
    ScaleA=true
    ScaleRGB=true
   )
   &DynamicMaterialParameters(
    Index 0 Param 1=&FloatFromCurve(FloatCurve=DataInterface() {propertyValues {Curve {keys [{tangentMode=RCTM_User, time=0, value=1}, {tangentMode=RCTM_User, time=1, value=1}]}}})
    Index 0 Param 2=0.7
    Index 0 Param 3=&FloatFromCurve(FloatCurve=DataInterface() {propertyValues {Curve {keys [
         {arriveTangent=4.0072, interpMode=RCIM_Cubic, leaveTangent=4.0072, tangentMode=RCTM_User, time=0, value=0}
         {arriveTangent=5, interpMode=RCIM_Cubic, leaveTangent=5, time=0.2543, value=0.6225}
         {arriveTangent=3.1884, interpMode=RCIM_Cubic, leaveTangent=3.1884, time=0.4, value=2}
         {arriveTangent=1.6993, interpMode=RCIM_Cubic, leaveTangent=1.6993, tangentMode=RCTM_User, time=1, value=3}
        ]}}})
    Index 0 Param 4=10
    Index 1 Param 1=&FN_FrontEndFloatModifier(Gameplay=4, Front End=2)
    Index 1 Param 2=$Particles.Random
    Index 1 Param 3=1
    Index 1 Param 4=1
    Index 2 Param 1=&FloatFromCurve(FloatCurve=DataInterface() {propertyValues {Curve {keys [<at 0 1>,<at 1 0.3>]}}})
    Module.Param0WriteEnabled=true
    Module.Param0WriteEnabled1=true
    Module.Param0WriteEnabled2=true
    Module.Param1WriteEnabled=true
    Module.Param1WriteEnabled1=true
    Module.Param1WriteEnabled2=true
    Module.Param2WriteEnabled=false
    Module.Param2WriteEnabled1=true
    Module.Param2WriteEnabled2=true
    Module.Param3WriteEnabled=true
    Module.Param3WriteEnabled1=true
    Module.Param3WriteEnabled2=true
    Write Parameter Index 1=true
    Write Parameter Index 2=true
   )
   &SolveForcesAndVelocity(Clamp Velocity=false, Limit Acceleration=false)
   /SetVariables(Particles.Lightning glow color:LinearColor=&MakeLinearColorFromVectorAndFloat(
     "Vector (RGB)"=&MakeVectorFromLinearColorRGB(Color=&MultiplyLinearColor(
       Linear Color=<rgba 0 0.4376 1 1>
       Scale Factor=&MakeVector4_FromFloat(Float=&RandomRangeFloat(Minimum=0.5, Maximum=1, Evaluation Type=::NewEnumerator1))
      ))
     "Float (Alpha)"=&Multiply_Float(
      A=&FloatFromCurve(
       FloatCurve=DataInterface() {propertyValues {Curve {keys [<at 0 1 RCIM_Cubic>, {interpMode=RCIM_Cubic, leaveTangent=-3.6307, time=1, value=1}]}}}
       Scale Curve=&FNTimeOfDayScalar(Night Scale=0.5, Morning Scale=0.7, Evening Scale=0.7)
      )
      B=3
     )
    )) {isSetParameters=true}
   /SetVariables(Particles.Glow Dy02:Vector4f=&MakeVector4(
     X=&FloatFromCurve(FloatCurve=DataInterface() {propertyValues {Curve {keys [{tangentMode=RCTM_User, time=0, value=1}, {tangentMode=RCTM_User, time=0.5, value=1}]}}})
     Y=0.7
     Z=0
     W=10
    )) {isSetParameters=true}
   /SetVariables(Particles.Glow Dy03:Vector4f=&MakeVector4(
     X=&FloatFromCurve(
      FloatCurve=DataInterface() {propertyValues {Curve {keys [<at 0 1>,<at 1 1>]}}}
      Scale Curve=&Multiply_Float(A=&FN_FrontEndFloatModifier(Gameplay=4, Front End=2), B=1)
     )
     Y=$Particles.Random
     Z=1
     W=1
    )) {isSetParameters=true}
  ]
  rendererClasses [&NiagaraMeshRendererProperties]
  renderers [
   {rendererClass=&NiagaraMeshRendererProperties, rendererIndex=0}
   {rendererClass=&NiagaraMeshRendererProperties, rendererIndex=1}
  ]
 }
 Lightning_OnMesh001 {
  emitterUpdateScript [
   &EmitterState(
    Inactive Response=::NewEnumerator1
    Life Cycle Mode=::NewEnumerator1
    Loop Delay=&RandomRangeFloat(Minimum=0.5, Maximum=3)
    Loop Duration=&RandomRangeFloat(Minimum=1, Maximum=3)
    Recalculate Duration Each Loop=true
    UseLoopDelay=false
   )
   &SpawnBurst_Instantaneous(
    Spawn Probability=&FN_FrontEndFloatModifier(Gameplay=$System.InterpolateOverTime004.InterpolateOverTime_Float.Moving Average, Front End=1)
    Use Spawn Probability=true
    UseLoopCountLimit=false
   )
   &SpawnRate(
    Spawn Probability=&FN_FrontEndFloatModifier(Gameplay=$System.InterpolateOverTime004.InterpolateOverTime_Float.Moving Average, Front End=1)
    SpawnRate=3
    Use Spawn Probability=true
   )
  ]
  particleSpawnScript [
   &InitializeParticle(
    Color=<rgba 0 0.1565 0.2 10>
    Lifetime Max=0.7
    Lifetime Min=0.2
    Lifetime Mode=::NewEnumerator1
    Mesh Scale=&RandomRangeVector(Minimum=<1 2 1.3>, Maximum=<1 1.5 1>)
    Mesh Scale Mode=::NewEnumerator4
    UsePositionOffset=false
    Write Material Random=true
   )
   &CameraOffset()
   &ShapeLocation(
    Cylinder Height=0
    Cylinder Radius=1
    MaskSpawnGroup=false
    Override Seed=false
    Rotation Axis=<0 1 0>
    Rotation Mode=::NewEnumerator1
    Shape Primitive=::NewEnumerator1
    Surface Only=false
    UseOwnerScale=true
   )
   &AddVelocity(Module.Use Falloff Distance=false, Override Seed=false, Velocity Mode=::NewEnumerator1, Velocity Speed=1)
   /SetVariables(Particles.Random=&RandomRangeFloat(Minimum=0, Maximum=1)) {isSetParameters=true, module=SetVariables_#10}
   &InitialMeshOrientation(
    Mesh Orientation Mode=::NewEnumerator2
    Rotation=&MakeCustomVectorFromBool(
     Bool=&RandomBool()
     True Vector=&RandomRangeVector(Minimum=<0 -0.13 0>, Maximum=<0 0.13 0>)
     False Vector=&RandomRangeVector(Minimum=<0 0.37 0>, Maximum=<0 0.63 0>)
    )
    Use Rotation Vector=true
   ) {module=InitialMeshOrientation001}
   &InitialMeshOrientation(Mesh Orientation Mode=::NewEnumerator2, Use Rotation Vector=true) {module=InitialMeshOrientation003}
  ]
  particleUpdateScript [
   &SkeletalMeshLocation(
    Bone Sampling Mode=::NewEnumerator4
    MaskSpawnGroup=false
    Sampled Position Offset=&AddVectorToPosition(Position=$System.FXMeshPosition, Vector=&RandomRangeVector(Minimum=<0 -100 0>, Maximum=<0 -250 0>))
    Skeletal Mesh=$System.Skeletal Mesh
   )
   &ParticleState()
   &ScaleMeshSize(Scale Factor=&Multiply_VectorByFloat(Vector=<3 3 3>, Float=&RandomRangeFloat(Minimum=0.8, Maximum=1.5)))
   &ScaleColor(
    Scale Alpha=&Multiply_Float(
     A=10
     B=&FloatFromCurve(FloatCurve=DataInterface() {propertyValues {Curve {keys [<at 0 0>,<at 0.1 0>,<at 0.2 0.8>,<at 1 0>]}}})
    )
    Scale RGB=&Multiply_VectorByFloat(
     Vector=<1 1 1>
     Float=&Multiply_Float(
      A=&RandomRangeFloat(Minimum=0, Maximum=1, Evaluation Type=::NewEnumerator1)
      B=&Multiply_Float(
       A=1
       B=&FloatFromCurve(FloatCurve=DataInterface() {propertyValues {Curve {keys [<at 0 10>,<at 1 10>]}}})
      )
     )
    )
    ScaleA=true
    ScaleRGB=true
   )
   &DynamicMaterialParameters(
    Index 0 Param 1=&FloatFromCurve(FloatCurve=DataInterface() {propertyValues {Curve {keys [{tangentMode=RCTM_User, time=0, value=0}, {tangentMode=RCTM_User, time=1, value=1}]}}})
    Index 0 Param 2=0.7
    Index 0 Param 3=&FloatFromCurve(FloatCurve=DataInterface() {propertyValues {Curve {keys [<at 0 0 RCIM_Cubic>, {arriveTangent=-0.3684, interpMode=RCIM_Cubic, leaveTangent=-0.3684, tangentMode=RCTM_User, time=1, value=3}]}}})
    Index 0 Param 4=2
    Index 1 Param 1=&FN_FrontEndFloatModifier(Gameplay=1.5, Front End=0.7)
    Index 1 Param 3=&Lerp_Float(A=1, B=1.5, Alpha=$Particles.Random)
    Index 1 Param 4=1
    Module.Param0WriteEnabled=true
    Module.Param0WriteEnabled1=true
    Module.Param1WriteEnabled=true
    Module.Param1WriteEnabled1=true
    Module.Param2WriteEnabled=true
    Module.Param2WriteEnabled1=true
    Module.Param3WriteEnabled=true
    Module.Param3WriteEnabled1=true
    Write Parameter Index 1=true
   )
   &SolveForcesAndVelocity(Clamp Velocity=false, Limit Acceleration=false)
   /SetVariables(Particles.Lightning glow color:LinearColor=&MakeLinearColorFromVectorAndFloat(
     "Vector (RGB)"=&MakeVectorFromLinearColorRGB(Color=&MultiplyLinearColor(
       Linear Color=<rgba 0 1.9641 3 1>
       Scale Factor=/Niagara/DynamicInputs/Multiply/Multiply_Vector4_ByFloat(Vector4:Vector4f=<1 1 1 1>, Float:NiagaraFloat=1)
      ))
     "Float (Alpha)"=&Multiply_Float(
      A=&FloatFromCurve(
       FloatCurve=DataInterface() {propertyValues {Curve {keys [
           {arriveTangent=-0.6023, interpMode=RCIM_Cubic, leaveTangent=-0.6023, tangentMode=RCTM_User, time=0, value=0}
           <at 0.1 0.85 RCIM_Cubic>
           {arriveTangent=-1.052, interpMode=RCIM_Cubic, leaveTangent=-1.052, tangentMode=RCTM_User, time=0.85, value=0}
          ]}}}
       Scale Curve=&FN_FrontEndFloatModifier(Gameplay=1, Front End=0.8)
      )
      B=3
     )
    )) {isSetParameters=true, module=SetVariables_#11}
   /SetVariables(
    Particles.Glow Dy02:Vector4f=&MakeVector4(
     X=&FloatFromCurve(FloatCurve=DataInterface() {propertyValues {Curve {keys [{tangentMode=RCTM_User, time=0, value=0}, {tangentMode=RCTM_User, time=1, value=1}]}}})
     Y=0.7
     Z=0
     W=2
    )
    Particles.Glow Dy03:Vector4f=&MakeVector4(
     X=&FN_FrontEndFloatModifier(Gameplay=1.5, Front End=0.7)
     Y=0
     Z=&Lerp_Float(A=1, B=1.5, Alpha=$Particles.Random)
     W=0
    )
   ) {isSetParameters=true, module=SetVariables_#12}
   &FN_Sprite_Attributes(
    Camera Offset=10
    Color=&MakeLinearColorFromVectorAndFloat(
     "Vector (RGB)"=&Multiply_Vector(A=&MakeVectorFromLinearColorRGB(Color=<rgba 1.25 0 5 1>), B=<1 1 1>)
     "Float (Alpha)"=&Multiply_Float(
      A=&FloatFromCurve(FloatCurve=DataInterface() {propertyValues {Curve {keys [
           {interpMode=RCIM_Cubic, tangentMode=RCTM_User, time=0, value=0}
           {interpMode=RCIM_Cubic, tangentMode=RCTM_User, time=0.3, value=1}
           {arriveTangent=-1.4286, interpMode=RCIM_Cubic, leaveTangent=-1.4286, time=0.7, value=1}
           {tangentMode=RCTM_User, time=1, value=0}
          ]}}})
      B=1
     )
    )
    Create Camera Offset Attribute=true
    Create Color Attribute=true
    Create Position Attribute=true
    Create SpriteSize Attribute=true
    Position:NiagaraPosition=&TransformPosition(
     Position=&AddVectorToPosition(
      Position=$Particles.Position
      Vector=&RotateVectorByQuaternion(
       Vector To Rotate=&RandomRangeVector(Minimum=<20 0 0>, Maximum=<35 0 0>)
       Quaternion=$Particles.SkeletalMeshLocation.SampledRotation
      )
     )
     Source Space=::Local
     Destination Space=::Simulation
    )
    SpriteSize=&Multiply_Vector2D_ByFloat(
     Float=&FloatFromCurve(
      FloatCurve=DataInterface() {propertyValues {Curve {keys [
          {interpMode=RCIM_Cubic, tangentMode=RCTM_User, time=0, value=0}
          {interpMode=RCIM_Cubic, tangentMode=RCTM_User, time=0.3, value=1}
          {tangentMode=RCTM_User, time=1, value=0}
         ]}}}
      Scale Curve=1.2
     )
     Vector2D=<30 30>
    )
    Write to Module Namespace=true
   )
  ]
  rendererClasses [&NiagaraMeshRendererProperties]
  renderers [
   {rendererClass=&NiagaraMeshRendererProperties, rendererIndex=0}
   {rendererClass=&NiagaraMeshRendererProperties, rendererIndex=1}
  ]
 }
 Lightning_Tail001 {
  emitterUpdateScript [
   &EmitterState(
    Inactive Response=::NewEnumerator1
    Life Cycle Mode=::NewEnumerator1
    Loop Delay=&RandomRangeFloat(Minimum=0.5, Maximum=3)
    Loop Duration=&RandomRangeFloat(Minimum=1, Maximum=3)
    Recalculate Duration Each Loop=true
    UseLoopDelay=false
   )
   &SpawnBurst_Instantaneous(
    Spawn Probability=$System.InterpolateOverTime004.InterpolateOverTime_Float.Moving Average
    Use Spawn Probability=true
    UseLoopCountLimit=false
   )
   &SpawnRate(Spawn Probability=$System.InterpolateOverTime004.InterpolateOverTime_Float.Moving Average, SpawnRate=5, Use Spawn Probability=true)
  ]
  particleSpawnScript [
   &PartitionParticles()
   &InitializeParticle(
    Color=<rgba 0 0.1567 0.2 10>
    Lifetime Max=0.7
    Lifetime Min=0.2
    Lifetime Mode=::NewEnumerator1
    Mesh Scale=&RandomRangeVector(Minimum=<1 2.5 1.3>, Maximum=<1 1.5 1>)
    Mesh Scale Mode=::NewEnumerator4
    UsePositionOffset=false
    Write Material Random=true
   )
   &CameraOffset()
   &ShapeLocation(
    Cylinder Height=0
    Cylinder Radius=1
    MaskSpawnGroup=false
    Override Seed=false
    Rotation Axis=<0 1 0>
    Rotation Mode=::NewEnumerator1
    Shape Primitive=::NewEnumerator1
    Surface Only=false
    UseOwnerScale=true
   )
   &AddVelocity(Module.Use Falloff Distance=false, Override Seed=false, Velocity Mode=::NewEnumerator1, Velocity Speed=1)
   /SetVariables(Particles.Random=&RandomRangeFloat(Minimum=0, Maximum=1)) {isSetParameters=true, module=SetVariables_#10}
   &InitialMeshOrientation(
    Mesh Orientation Mode=::NewEnumerator2
    Rotation=&RandomRangeVector(
     Minimum=&SelectVectorFromArray(
      Direct Array Index=$Particles.PartitionParticles.Partition
      Vector Selection Array=DataInterface() {propertyValues {FloatData [<0 -0.1 0>, <0 0.4 0>]}}
     )
     Maximum=&SelectVectorFromArray(
      Direct Array Index=$Particles.PartitionParticles.Partition
      Vector Selection Array=DataInterface() {propertyValues {FloatData [<0 0.1 0>, <0 0.6 0>]}}
     )
    )
    Use Rotation Vector=true
   ) {module=InitialMeshOrientation003}
   &InitialMeshOrientation(
    Mesh Orientation Mode=::NewEnumerator2
    Rotation=&SelectVectorFromArray(
     Direct Array Index=$Particles.PartitionParticles.Partition
     Vector Selection Array=DataInterface() {propertyValues {FloatData [<0 0 -0.05>, <0 0 0.05>]}}
    )
    Use Rotation Vector=true
   )
  ]
  particleUpdateScript [
   &SkeletalMeshLocation(
    Bone Sampling Mode=::NewEnumerator4
    MaskSpawnGroup=false
    Sampled Position Offset=&AddVectorToPosition(
     Position=$System.FXMeshPosition
     Vector=&RandomRangeVector(
      Minimum=&SelectVectorFromArray(
       Direct Array Index=$Particles.PartitionParticles.Partition
       Vector Selection Array=DataInterface() {propertyValues {FloatData [<50 80 0>, <-50 80 0>]}}
      )
      Maximum=&SelectVectorFromArray(
       Direct Array Index=$Particles.PartitionParticles.Partition
       Vector Selection Array=DataInterface() {propertyValues {FloatData [<10 80 0>, <-10 80 0>]}}
      )
     )
    )
    Skeletal Mesh=$System.Skeletal Mesh
   )
   &ParticleState()
   &ScaleMeshSize(Scale Factor=&Multiply_VectorByFloat(Vector=<3 3 2>, Float=&RandomRangeFloat(Minimum=0.8, Maximum=1)))
   &ScaleColor(
    Scale Alpha=&Multiply_Float(
     A=10
     B=&FloatFromCurve(FloatCurve=DataInterface() {propertyValues {Curve {keys [<at 0 0>,<at 0.1 0>,<at 0.2 0.8>,<at 1 0>]}}})
    )
    Scale RGB=&Multiply_VectorByFloat(
     Vector=<1 1 1>
     Float=&Multiply_Float(
      A=&RandomRangeFloat(Minimum=0, Maximum=1, Evaluation Type=::NewEnumerator1)
      B=&Multiply_Float(
       A=1
       B=&FloatFromCurve(FloatCurve=DataInterface() {propertyValues {Curve {keys [<at 0 10>,<at 1 10>]}}})
      )
     )
    )
    ScaleA=true
    ScaleRGB=true
   )
   &DynamicMaterialParameters(
    Index 0 Param 1=&FloatFromCurve(FloatCurve=DataInterface() {propertyValues {Curve {keys [{tangentMode=RCTM_User, time=0, value=0}, {tangentMode=RCTM_User, time=1, value=1}]}}})
    Index 0 Param 2=0.7
    Index 0 Param 3=&FloatFromCurve(FloatCurve=DataInterface() {propertyValues {Curve {keys [<at 0 0 RCIM_Cubic>, {arriveTangent=-0.3684, interpMode=RCIM_Cubic, leaveTangent=-0.3684, tangentMode=RCTM_User, time=1, value=3}]}}})
    Index 0 Param 4=2
    Index 1 Param 1=&FN_FrontEndFloatModifier(Gameplay=1.5, Front End=0.7)
    Index 1 Param 2=&Lerp_Float(A=0, B=1, Alpha=$Particles.Random)
    Index 1 Param 3=&Lerp_Float(A=1, B=1.5, Alpha=$Particles.Random)
    Index 1 Param 4=1
    Module.Param0WriteEnabled=true
    Module.Param0WriteEnabled1=true
    Module.Param1WriteEnabled=true
    Module.Param1WriteEnabled1=true
    Module.Param2WriteEnabled=true
    Module.Param2WriteEnabled1=true
    Module.Param3WriteEnabled=true
    Module.Param3WriteEnabled1=true
    Write Parameter Index 1=true
   )
   &SolveForcesAndVelocity(Clamp Velocity=false, Limit Acceleration=false)
   /SetVariables(Particles.Lightning glow color:LinearColor=&MakeLinearColorFromVectorAndFloat(
     "Vector (RGB)"=&MakeVectorFromLinearColorRGB(Color=&MultiplyLinearColor(
       Linear Color=<rgba 0 1.9666 3 1>
       Scale Factor=&Multiply_Vector4_ByFloat(Vector4:Vector4f=<1 1 1 1>, Float:NiagaraFloat=1)
      ))
     "Float (Alpha)"=&Multiply_Float(
      A=&FloatFromCurve(FloatCurve=DataInterface() {propertyValues {Curve {keys [
           {arriveTangent=-0.6023, interpMode=RCIM_Cubic, leaveTangent=-0.6023, tangentMode=RCTM_User, time=0, value=0}
           <at 0.1 0.85 RCIM_Cubic>
           {arriveTangent=-1.052, interpMode=RCIM_Cubic, leaveTangent=-1.052, tangentMode=RCTM_User, time=0.85, value=0}
          ]}}})
      B=2
     )
    )) {isSetParameters=true, module=SetVariables_#11}
   /SetVariables(
    Particles.Glow Dy02:Vector4f=&MakeVector4(
     X=&FloatFromCurve(FloatCurve=DataInterface() {propertyValues {Curve {keys [{tangentMode=RCTM_User, time=0, value=0}, {tangentMode=RCTM_User, time=1, value=1}]}}})
     Y=0.7
     Z=0
     W=2
    )
    Particles.Glow Dy03:Vector4f=&MakeVector4(
     X=&FN_FrontEndFloatModifier(Gameplay=1.5, Front End=0.7)
     Y=&Lerp_Float(A=0, B=1, Alpha=$Particles.Random)
     Z=&Lerp_Float(A=1, B=1.5, Alpha=$Particles.Random)
     W=0
    )
   ) {isSetParameters=true, module=SetVariables_#12}
   &FN_Sprite_Attributes(
    Camera Offset=10
    Color=&MakeLinearColorFromVectorAndFloat(
     "Vector (RGB)"=&Multiply_Vector(A=&MakeVectorFromLinearColorRGB(Color=<rgba 1.25 0 5 1>), B=<1 1 1>)
     "Float (Alpha)"=&Multiply_Float(
      A=&FloatFromCurve(FloatCurve=DataInterface() {propertyValues {Curve {keys [
           {interpMode=RCIM_Cubic, tangentMode=RCTM_User, time=0, value=0}
           {interpMode=RCIM_Cubic, tangentMode=RCTM_User, time=0.3, value=1}
           {arriveTangent=-1.4286, interpMode=RCIM_Cubic, leaveTangent=-1.4286, time=0.7, value=1}
           {tangentMode=RCTM_User, time=1, value=0}
          ]}}})
      B=1
     )
    )
    Create Camera Offset Attribute=true
    Create Color Attribute=true
    Create Position Attribute=true
    Create SpriteSize Attribute=true
    Position:NiagaraPosition=&TransformPosition(
     Position=&AddVectorToPosition(
      Position=$Particles.Position
      Vector=&RotateVectorByQuaternion(
       Vector To Rotate=&RandomRangeVector(Minimum=<20 0 0>, Maximum=<35 0 0>)
       Quaternion=$Particles.SkeletalMeshLocation.SampledRotation
      )
     )
     Source Space=::Local
     Destination Space=::Simulation
    )
    SpriteSize=&Multiply_Vector2D_ByFloat(
     Float=&FloatFromCurve(
      FloatCurve=DataInterface() {propertyValues {Curve {keys [
          {interpMode=RCIM_Cubic, tangentMode=RCTM_User, time=0, value=0}
          {interpMode=RCIM_Cubic, tangentMode=RCTM_User, time=0.3, value=1}
          {tangentMode=RCTM_User, time=1, value=0}
         ]}}}
      Scale Curve=1.2
     )
     Vector2D=<30 30>
    )
    Write to Module Namespace=true
   )
  ]
  rendererClasses [&NiagaraMeshRendererProperties]
  renderers [
   {rendererClass=&NiagaraMeshRendererProperties, rendererIndex=0}
   {rendererClass=&NiagaraMeshRendererProperties, rendererIndex=1}
  ]
 }
 Smoke {
  emitterUpdateScript [
   &EmitterState()
   &SpawnRate(
    Spawn Probability=&Lerp_Float(A=0.5, B=1, Alpha=$System.InterpolateOverTime004.InterpolateOverTime_Float.Moving Average)
    SpawnRate=25
    Use Spawn Probability=true
   )
  ]
  particleSpawnScript [
   &PartitionParticles()
   &InitializeParticle(
    Lifetime Max=1.3
    Lifetime Min=0.7
    Lifetime Mode=::NewEnumerator1
    Sprite Rotation Mode=::NewEnumerator2
    Sprite Size Mode=::NewEnumerator1
    Uniform Sprite Size Max=100
    Uniform Sprite Size Min=200
    UsePositionOffset=false
    Write Material Random=true
   )
   &SkeletalMeshLocation(
    Bone Sampling Mode=::NewEnumerator4
    Interpolate Mesh Sampling=false
    MaskSpawnGroup=false
    Orientation Sampling=::NewEnumerator1
    Sampled Position Offset=&AddVectorToPosition(
     Position=&SelectVectorFromArray(
      Direct Array Index=$Particles.PartitionParticles.Partition
      Vector Selection Array=DataInterface() {propertyValues {FloatData [<-25 -20 -7>, <20 0 -7>]}}
     )
     Vector=$System.FX position offset
    )
    Skeletal Mesh=$System.Skeletal Mesh
   )
   &AddVelocity(
    Scale Added Velocity=&Multiply_VectorByFloat(
     Vector=&Multiply_VectorByFloat(
      Vector=<2.5 2.5 2.5>
      Float=&Lerp_Float(A=0.3, B=1, Alpha=$System.InterpolateOverTime004.InterpolateOverTime_Float.Moving Average)
     )
     Float=&Lerp_Float(A=1, B=0.6, Alpha=$System.InterpolateOverTime.InterpolateOverTime_Float.Moving Average)
    )
    Velocity=&SelectVectorFromArray(
     Direct Array Index=$Particles.PartitionParticles.Partition
     Vector Selection Array=DataInterface() {propertyValues {FloatData [<-25 120 0>, <25 120 0>]}}
    )
   )
   &AddVelocityInCone(Cone Angle=30, Cone Axis=<0 1 0>, Module.Use Velocity Falloff On Cone Axis=false, Velocity Strength=100) {module=AddVelocityInCone001}
   &InitialMeshOrientation(
    Mesh Orientation Mode=::NewEnumerator2
    Rotation=&RandomRangeVector(Minimum=<0 0 0>, Maximum=<1 0 0>)
    Use Rotation Vector=true
   )
   &CameraOffset(Camera Offset Amount=-20)
  ]
  particleUpdateScript [
   &Drag(
    Drag=&Multiply_Float(
     A=&MakeCustomFloatFromBool(
      Bool=&SetBoolByIntComparison(A=$Particles.PartitionParticles.Partition, B=0, Comparison Type=::NewEnumerator0)
      True Float=&Lerp_Float(A=0, B=1, Alpha=$System.InterpolateOverTime001.InterpolateOverTime_Float.Moving Average)
      False Float=&Lerp_Float(A=0, B=1, Alpha=$System.InterpolateOverTime002.InterpolateOverTime_Float.Moving Average)
     )
     B=1
    )
    Use Linear Drag=true
    Use Rotational Drag=true
   )
   &SolveForcesAndVelocity(Clamp Velocity=false, Limit Acceleration=false)
   &ParticleState()
   &ScaleSpriteSize(
    Uniform Curve Scale=&Lerp_Float(A=0.8, B=1, Alpha=$System.InterpolateOverTime004.InterpolateOverTime_Float.Moving Average)
    Uniform Curve Sprite Scale=DataInterface() {propertyValues {Curve {keys [
        {arriveTangent=1.9933, interpMode=RCIM_Cubic, leaveTangent=1.9933, tangentMode=RCTM_User, time=0, value=0.3}
        {arriveTangent=1.2, interpMode=RCIM_Cubic, leaveTangent=1.2, time=0.2507, value=0.5735}
        {tangentMode=RCTM_User, time=1, value=1.5}
       ]}}}
   )
   &SpriteRotationRate()
   &ScaleColor(
    Scale Alpha=&FloatFromCurve(
     FloatCurve=DataInterface() {propertyValues {Curve {keys [
         {arriveTangent=2.8927, interpMode=RCIM_Cubic, leaveTangent=2.8927, tangentMode=RCTM_User, time=0, value=0}
         <at 0.3 1>
         {arriveTangent=-1.4286, interpMode=RCIM_Cubic, leaveTangent=-1.4286, time=0.7646, value=0.7176}
         <at 1 0>
        ]}}}
     Scale Curve=&FN_FrontEndFloatModifier(Gameplay=1.5, Front End=0.7)
    )
    Scale RGB=&MakeVectorFromLinearColorRGB(Color=&ColorFromCurve(ColorCurve=DataInterface() {propertyValues {
        AlphaCurve {keys []}
        BlueCurve {keys [
          {arriveTangent=-331.6383, interpMode=RCIM_Cubic, time=0.0006, value=2}
          <at 0.2634 3>
          <at 0.3441 1>
          <at 0.4944 1.4331>
          {arriveTangent=-1.8989, interpMode=RCIM_Cubic, leaveTangent=-1.8989, time=0.7598, value=1.3}
          <at 0.9858 0.5>
         ]}
        GreenCurve {keys [
          {arriveTangent=-87.3695, interpMode=RCIM_Cubic, time=0.0006, value=1.4618}
          <at 0.2634 0>
          <at 0.3441 0.6463>
          <at 0.4944 0.0078>
          {arriveTangent=-0.0158, interpMode=RCIM_Cubic, leaveTangent=-0.0158, time=0.7598, value=0}
          <at 0.9858 0>
         ]}
        RedCurve {keys [
          <at 0.0006 0 RCIM_Cubic>
          <at 0.2634 1.1721>
          <at 0.3441 0>
          <at 0.4944 0.4685>
          {arriveTangent=-0.1031, interpMode=RCIM_Cubic, leaveTangent=-0.1031, time=0.7598, value=1.0865}
          <at 0.9858 0.4179>
         ]}
       }}))
    ScaleA=true
    ScaleRGB=true
   )
   &DynamicMaterialParameters(Module.Index0Float0Write=true, Module.Index0Float1Write=true, Module.Index0Float2Write=true, Module.Index0Float3Write=true) {module=DynamicMaterialParameters001}
  ]
  rendererClasses [&NiagaraSpriteRendererProperties]
  renderers [{rendererClass=&NiagaraSpriteRendererProperties}]
 }
 SpikesShape001 {
  emitterUpdateScript [
   /SetVariables(Emitter.RandomFloat:NiagaraFloat=&RandomRangeFloat(Minimum=0.3, Maximum=1)) {isSetParameters=true, module=SetVariables_#7}
   &EmitterState(
    Life Cycle Mode=::NewEnumerator1
    Loop Behavior=::NewEnumerator1
    Loop Delay=$System.Delay
    Loop Duration Mode=::NewEnumerator1
    UseLoopDelay=true
   )
   &SpawnBurst_Instantaneous(Use Spawn Probability=true, UseLoopCountLimit=false)
  ]
  particleSpawnScript [
   &InitializeParticle(
    Color=<rgba 0.0781 0.9647 3 2>
    Mesh Scale=<1.6 2.1 2.2>
    Mesh Scale Mode=::NewEnumerator4
    UsePositionOffset=false
    Write Material Random=true
   )
   &InitialMeshOrientation(Mesh Orientation Mode=::NewEnumerator2, Rotation=<0 0 0.25>, Use Rotation Vector=true)
   &InitialMeshOrientation(Mesh Orientation Mode=::NewEnumerator2, Rotation=<0 0 -0.01>, Use Rotation Vector=true) {module=InitialMeshOrientation001}
   &CameraOffset(Camera Offset Amount=-10)
  ]
  particleUpdateScript [
   /SetVariables(Particles.FlashColor:LinearColor=&LerpLinearColors(StartColor=<rgba 1 1 1 1>, EndColor=<rgba 0 0.0722 0.401 1>, LerpFactor=0)) {isSetParameters=true, module=SetVariables_#8}
   &SkeletalMeshLocation(
    Bone Sampling Mode=::NewEnumerator4
    MaskSpawnGroup=false
    Sampled Position Offset=&AddVectorToPosition(
     Position=$System.FXMeshPosition
     Vector=&FN_FrontEndVector3Modifier(Gameplay=<15 0 -13>, Front End=<15 0 -13>)
    )
    Skeletal Mesh=$System.Skeletal Mesh
   )
   &ParticleState(Kill Particles When Lifetime Has Elapsed=false, Let Infinitely Lived Particles Die When Emitter Deactivates=true)
   &ScaleMeshSize(Scale Factor=&Multiply_VectorByFloat(Vector=<1 1.1 1>, Float=&Multiply_Float(A=0.8, B=$System.SizeScale)))
   &ScaleColor(
    Scale Alpha=&Multiply_Float(A=&Lerp_Float(A=0, B=1, Alpha=$System.InterpolateOverTime004.InterpolateOverTime_Float.Moving Average), B=1)
    Scale RGB=&Multiply_VectorByFloat(
     Vector=<1 1 1>
     Float=&Multiply_Float(A=1, B=&FN_PickFloatDependOnPlatform(TargetPlatformActive=1, TargetPlatformNoActive=0.6))
    )
    ScaleA=true
    ScaleRGB=true
   ) {enabled=false}
   &FN_Sprite_Attributes(
    Color=&MakeLinearColorFromVectorAndFloat(
     "Vector (RGB)"=&MakeVectorFromLinearColorRGB(Color=&MultiplyLinearColor(
       Linear Color=<rgba 0 0.8239 1 2>
       Scale Factor=&MakeVector4_FromFloat(Float=&Multiply_Float(A=1.5, B=&FN_PickFloatDependOnPlatform(TargetPlatformActive=1, TargetPlatformNoActive=0.8)))
      ))
     "Float (Alpha)"=&Lerp_Float(
      A=0
      B=&FN_FrontEndFloatModifier(Gameplay=2, Front End=0.5)
      Alpha=$System.InterpolateOverTime004.InterpolateOverTime_Float.Moving Average
     )
    )
    Create Color Attribute=true
    Write to Module Namespace=true
   ) {module=FN_Sprite_Attributes001}
   &DynamicMaterialParameters(
    Index 0 Param 2=1
    Index 0 Param 3=&Add_Float(
     A=&FloatFromCurve(
      FloatCurve=DataInterface() {propertyValues {Curve {keys [
          {arriveTangent=0.5715, interpMode=RCIM_Cubic, leaveTangent=0.5715, tangentMode=RCTM_User, time=0, value=0}
          {arriveTangent=1.1179, interpMode=RCIM_Cubic, leaveTangent=1.1179, tangentMode=RCTM_User, time=0.5, value=0.6}
          {arriveTangent=0.5282, interpMode=RCIM_Cubic, leaveTangent=0.5282, tangentMode=RCTM_User, time=1, value=1}
         ]}}}
      Scale Curve=0.95
     )
     B=1
    )
    Index 0 Param 4=&UniformAOrBFloat(A=-1, Evaluation Type=::NewEnumerator0, Recalculate Random Each Loop=true, Override Seed=false)
    Index 3 Param 1=0.3
    Module.Index0Param0Write=true
    Module.Index0Param1Write=true
    Module.Index0Param2Write=true
    Module.Index0Param3Write=true
    Module.Index3Param0Write=true
    Module.Index3Param1Write=false
    Module.Index3Param2Write=false
    Module.Index3Param3Write=false
    Write Parameter Index 3=true
   )
   /SetVariables(Particles.CenterPosition:NiagaraPosition=&AddVectorToPosition(Position=&AddVectorToPosition(Position=$Particles.Position, Vector=<-12 0 13>), Vector=<0 0 0>)) {isSetParameters=true, module=SetVariables_#9}
   &FN_Mesh_Attributes(
    Color=&MakeLinearColorFromVectorAndFloat(
     "Vector (RGB)"=&MakeVectorFromLinearColorRGB(Color=&MultiplyLinearColor(
       Linear Color=<rgba 0 0.849 1 1>
       Scale Factor=&MakeVector4_FromLinearColor(LinearColor=&MakeLinearColorFromFloat(Value:NiagaraFloat=&Multiply_Float(A=0.5, B=1)))
      ))
     "Float (Alpha)"=&FN_FrontEndFloatModifier(
      Gameplay=&Lerp_Float(A=0, B=0.5, Alpha=$System.InterpolateOverTime004.InterpolateOverTime_Float.Moving Average)
      Front End=0.3
     )
    )
    Create Color Attribute=true
    Create Scale Attribute=true
    Scale=&Multiply_VectorByFloat(Vector=<1.5 2 1>, Float=&Multiply_Float(A=1.1, B=$System.SizeScale))
    Write to Module Namespace=true
   )
   &FN_Mesh_Attributes(
    Color=&MakeLinearColorFromVectorAndFloat(
     "Vector (RGB)"=&MakeVectorFromLinearColorRGB(Color=&MultiplyLinearColor(
       Linear Color=<rgba 0.3546 0 1 1>
       Scale Factor=&MakeVector4_FromFloat(Float=&Multiply_Float(
         A=&FN_FrontEndFloatModifier(Gameplay=1, Front End=2)
         B=&FN_PickFloatDependOnPlatform(TargetPlatformActive=1, TargetPlatformNoActive=0.6)
        ))
      ))
     "Float (Alpha)"=&Lerp_Float(
      A=0
      B=&FN_FrontEndFloatModifier(Gameplay=1, Front End=1)
      Alpha=$System.InterpolateOverTime004.InterpolateOverTime_Float.Moving Average
     )
    )
    Create Color Attribute=true
    Create Scale Attribute=true
    Scale=&Multiply_Vector(
     A=<1.6 2.1 2.2>
     B=&Multiply_VectorByFloat(Vector=<1.1 1 1>, Float=&Multiply_Float(A=0.7, B=$System.SizeScale))
    )
    Write to Module Namespace=true
   ) {module=FN_Mesh_Attributes001}
  ]
  rendererClasses [&NiagaraMeshRendererProperties]
  renderers [
   {rendererClass=&NiagaraMeshRendererProperties, rendererIndex=0}
   {rendererClass=&NiagaraMeshRendererProperties, rendererIndex=1}
   {rendererClass=&NiagaraMeshRendererProperties, rendererIndex=2}
  ]
 }
 SplitLightning_OnSmoke001 {
  emitterUpdateScript [
   &EmitterState()
   &SpawnRate(
    Spawn Probability=&ClampFloat(Float=$User.ForwardDot, Min=&FN_FrontEndFloatModifier(Gameplay=0, Front End=1))
    SpawnRate=20
    Use Spawn Probability=true
   )
   &SpawnBurst_Instantaneous(
    Spawn Count=2
    Spawn Probability=&ClampFloat(Float=$User.ForwardDot, Min=&FN_FrontEndFloatModifier(Gameplay=0, Front End=1))
    Spawn Time=0.005
    Use Spawn Probability=true
    UseLoopCountLimit=false
   )
  ]
  particleSpawnScript [
   &PartitionParticles(Number Of Partitions=1)
   &InitializeParticle(
    Color=&MultiplyLinearColor(
     Linear Color=&RandomRangeLinearColor(Minimum=<rgba 1 1 1 0.3>, Maximum=<rgba 1.2 1.2 1.2 1.5>)
     Scale Factor=&MakeVector4_FromFloat(Float=1)
    )
    Lifetime=&RandomRangeFloat(Minimum=0.2, Maximum=0.3)
    Mesh Scale Mode=::NewEnumerator3
    Mesh Uniform Scale=&RandomRangeFloat(Minimum=0.8, Maximum=1.3)
    UsePositionOffset=false
    Write Material Random=true
   )
   &SkeletalMeshLocation(
    Bone Sampling Mode=::NewEnumerator4
    Interpolate Mesh Sampling=false
    MaskSpawnGroup=false
    Orientation Sampling=::NewEnumerator1
    Sampled Position Offset=&AddVectorToPosition(Position=$System.FX position offset, Vector=<0 0 0>)
    Skeletal Mesh=$System.Skeletal Mesh
   )
   &ShapeLocation(
    MaskSpawnGroup=false
    Non Uniform Scale=<1 1 0>
    Offset Mode=::NewEnumerator0
    Override Seed=false
    Shape Origin=&VectorFromFloat(Value=0)
    Sphere Radius=12
    UseOwnerScale=true
   )
   &InitialMeshOrientation(
    Mesh Axis To Orient=$Particles.Velocity
    Mesh Orientation Mode=::NewEnumerator2
    Orientation Vector=<0 0 -1>
    Rotation=<-0.25 0 0>
    Use Rotation Vector=false
   )
   &AddVelocity(
    Cone Angle=30
    Cone Axis=<0 1 -0.05>
    Module.UseConeFalloff=false
    Module.UseVelDistribution=false
    Override Seed=false
    Velocity Mode=::NewEnumerator2
    Velocity Speed=400
   )
   &SystemLocation(MaskSpawnGroup=false, Offset=<0 0 -7>)
   &CameraOffset(Camera Offset Amount=-50)
  ]
  particleUpdateScript [
   &ParticleState()
   &ScaleMeshSize(Scale Factor=&RandomRangeVector(Minimum=<1 1 2>, Maximum=<1.5 1.8 2>)) {module=ScaleMeshSize001}
   &ScaleMeshSize(Scale Factor=&MakeVector(
     X=1
     Y=1
     Z=&FloatFromCurve(FloatCurve=DataInterface() {propertyValues {Curve {keys [
          {interpMode=RCIM_Cubic, leaveTangent=0.0966, tangentMode=RCTM_Break, time=0, value=1}
          {arriveTangent=0.0951, interpMode=RCIM_Cubic, leaveTangent=-0.1745, tangentMode=RCTM_Break, time=1, value=1.07}
         ]}}})
    ))
   &DynamicMaterialParameters(
    Index 0 Param 2=&RandomRangeFloat(Minimum=0, Maximum=2)
    Index 0 Param 4=&MakeCustomFloatFromBool(
     Bool=false
     True Float=11
     False Float=&SelectFloatFromArray(
      Array Sampling Mode=::NewEnumerator0
      Direct Array Index=?
      Float Selection Array=DataInterface() {propertyValues {FloatData [3,11,1,9,2,11,11]}}
      Random Seed=0
     )
    )
    Index 1 Param 1=&FloatFromCurve(FloatCurve=DataInterface() {propertyValues {Curve {keys [<at 0 1>,<at 1 1.3>]}}}, Scale Curve=0.5)
    Index 1 Param 2=2
    Index 1 Param 3=&FloatFromCurve(FloatCurve=DataInterface() {propertyValues {Curve {keys [{interpMode=RCIM_Cubic, tangentMode=RCTM_Break, time=0.5, value=-0.5}, <at 1 0.1>]}}})
    Index 2 Param 1=&Multiply_Float(
     A=&FloatFromCurve(
      FloatCurve=DataInterface() {propertyValues {Curve {keys [{interpMode=RCIM_Cubic, leaveTangent=8.1085, tangentMode=RCTM_Break, time=0.8568, value=1}, {tangentMode=RCTM_User, time=1, value=0}]}}}
      Scale Curve=&MakeCustomFloatFromBool(Bool=$User.bIsFrontEndPreview, True Float=0.3, False Float=&RandomRangeFloat(Minimum=0.5, Maximum=0.6))
     )
     B=&Divide_Float(
      A=&MakeFloatFromVector(Vector=$Particles.Scale, Channel=::NewEnumerator0)
      B=&RandomRangeFloat(Minimum=10, Maximum=15)
     )
    )
    Index 2 Param 2=&FloatFromCurve(
     FloatCurve=DataInterface() {propertyValues {Curve {keys [
         {interpMode=RCIM_Cubic, tangentMode=RCTM_Break, time=0.7, value=0}
         {arriveTangent=-0.0859, interpMode=RCIM_Cubic, tangentMode=RCTM_Break, time=1, value=-1}
        ]}}}
     Scale Curve=50
    )
    Index 2 Param 3=&Waveform(
     Global Input Position=&RandomRangeFloat(Minimum=0, Maximum=15)
     Global Phase=$Particles.Age
     "Waveform [1]"=::NewEnumerator12
     "[1] Phase Scale"=2
     "[1] Frequency"=1
     "[1] Amplitude Min/Max"=<0 1>
    )
    Index 2 Param 4=25
    Index 3 Param 1=&FloatFromCurve(
     FloatCurve=DataInterface() {propertyValues {Curve {keys [
         {interpMode=RCIM_Cubic, leaveTangent=-0.0025, tangentMode=RCTM_Break, time=0, value=0}
         {arriveTangent=1.8528, interpMode=RCIM_Cubic, tangentMode=RCTM_Break, time=1, value=1}
        ]}}}
     Scale Curve=&RandomRangeFloat(Minimum=15, Maximum=25)
    )
    Index 3 Param 2=&FloatFromCurve(FloatCurve=DataInterface() {propertyValues {Curve {keys [
         {interpMode=RCIM_Cubic, leaveTangent=-0.0025, tangentMode=RCTM_Break, time=0, value=0.5}
         {arriveTangent=-0.0316, interpMode=RCIM_Cubic, tangentMode=RCTM_Break, time=1, value=-0.2}
        ]}}})
    Index 3 Param 3=&FloatFromCurve(FloatCurve=DataInterface() {propertyValues {Curve {keys [
         {interpMode=RCIM_Cubic, leaveTangent=1.1034, tangentMode=RCTM_Break, time=0, value=-1}
         {arriveTangent=2.2399, tangentMode=RCTM_User, time=0.5, value=0.125}
        ]}}})
    Index 3 Param 4=&FloatFromCurve(FloatCurve=DataInterface() {propertyValues {Curve {keys [
         {interpMode=RCIM_Cubic, leaveTangent=3.0601, tangentMode=RCTM_Break, time=0.5, value=-0.5}
         {arriveTangent=3.2053, tangentMode=RCTM_User, time=1, value=1}
        ]}}})
    Module.Index0Param0Write=true
    Module.Index0Param1Write=true
    Module.Index0Param2Write=true
    Module.Index0Param3Write=true
    Module.Index1Param0Write=true
    Module.Index1Param1Write=true
    Module.Index1Param2Write=true
    Module.Index1Param3Write=true
    Module.Index2Param0Write=true
    Module.Index2Param1Write=true
    Module.Index2Param2Write=true
    Module.Index2Param3Write=true
    Module.Index3Param0Write=true
    Module.Index3Param1Write=true
    Module.Index3Param2Write=true
    Module.Index3Param3Write=true
    Write Parameter Index 1=true
    Write Parameter Index 2=true
    Write Parameter Index 3=true
   ) {module=DynamicMaterialParameters001}
   &Drag(Drag=2, Use Linear Drag=true, Use Rotational Drag=true)
   &SolveForcesAndVelocity(Clamp Velocity=false, Limit Acceleration=false)
   &ScaleColor(
    Scale Alpha=&FloatFromCurve(FloatCurve=DataInterface() {propertyValues {Curve {keys [
         {tangentMode=RCTM_User, time=0, value=1}
         {arriveTangent=-0.0405, interpMode=RCIM_Cubic, leaveTangent=-0.0405, tangentMode=RCTM_User, time=0.7009, value=1}
         {arriveTangent=-8.0089, interpMode=RCIM_Cubic, leaveTangent=-8.0089, tangentMode=RCTM_User, time=1, value=0}
        ]}}})
    Scale RGB=&Lerp_Vector(
     A=&VectorFromFloat(Value=&FloatFromCurve(FloatCurve=DataInterface() {propertyValues {Curve {keys [<at 0 0>,<at 0.3786 0.2>,<at 1 20>]}}}, Scale Curve=20))
     B=&VectorFromFloat(Value=&RandomRangeFloat(Minimum=1, Maximum=1, Evaluation Type=::NewEnumerator1))
     Alpha=&FloatFromCurve(FloatCurve=DataInterface() {propertyValues {Curve {keys [
          {interpMode=RCIM_Cubic, leaveTangent=2.5294, tangentMode=RCTM_Break, time=0, value=0}
          {arriveTangent=-0.2166, interpMode=RCIM_Cubic, tangentMode=RCTM_Break, time=0.4, value=1}
         ]}}})
    )
    ScaleA=true
    ScaleRGB=true
   ) {module=ScaleColor004}
   /SetVariables(Particles.MeshOrientation=&MultiplyQuaternion(
     Quaternion A=&MultiplyQuaternion(
      Quaternion A=&MakeQuatFromAxisAngle(Axis=<0 1 0>, AngleInDegrees=0)
      Quaternion B=&FindQuatBetweenTwoVectors(From Vector=<0 0 1>, To Vector=$Particles.Velocity)
     )
     Quaternion B=$Particles.SkeletalMeshLocation.SampledRotation
    )) {isSetParameters=true}
  ]
  rendererClasses [&NiagaraMeshRendererProperties]
  renderers [
   {rendererClass=&NiagaraMeshRendererProperties, rendererIndex=0}
   {rendererClass=&NiagaraMeshRendererProperties, rendererIndex=1}
  ]
 }
}
summary {
 systemName=NS_Glider_BentBaton_FlipAway_Idle
 emitters [
  {emitterName=Cloud, rendererClasses [&NiagaraMeshRendererProperties]}
  {emitterName=CloudRing, rendererClasses [&NiagaraMeshRendererProperties]}
  {emitterName=Smoke, rendererClasses [&NiagaraSpriteRendererProperties]}
  {
   emitterName=FXMesh
   rendererClasses [
    &NiagaraMeshRendererProperties
    &NiagaraLightRendererProperties
    &NiagaraSpriteRendererProperties
   ]
  }
  {emitterName=SpikesShape001, rendererClasses [&NiagaraMeshRendererProperties]}
  {emitterName=SplitLightning_OnSmoke001, rendererClasses [&NiagaraMeshRendererProperties]}
  {emitterName=Lightning_OnMesh001, rendererClasses [&NiagaraMeshRendererProperties]}
  {emitterName=Lightning_Tail001, rendererClasses [&NiagaraMeshRendererProperties]}
  {emitterName=LightningBlur001, rendererClasses [&NiagaraMeshRendererProperties]}
  {emitterName=Frontend_SideSpikes, rendererClasses [&NiagaraMeshRendererProperties]}
 ]
 userVariables [
  User.bIsFrontEnd {type=NiagaraBool}
  User.bIsFrontEndPreview {type=NiagaraBool}
  User.ForwardDot {type=NiagaraFloat, defaultValue=0, description=""}
 ]
}
